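# Physics asset for the bracelets on the F_MED_Cavalry female player body cosmetic, defining bodies for their secondary motion.
# The top and bottom bracelets each pair a tiny zero-length capsule on a base bone with a box on the matching parent bone.
# One constraint per bracelet joins the two bodies with small swing and tighter twist limits, and no constraint links the two bracelets.
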PhysicsAsset /BRCosmetics/Characters/Player/Female/Medium/Bodies/F_MED_Cavalry/Meshes/F_MED_Cavalry_Bracelets_Physics
bodies {
 dyn_bracelet_bottom_base [{shapeType=Capsule, length=0, radius=0.5}]
 dyn_bracelet_bottom_parent [{shapeType=Box, extentX=1.3624, extentY=5.8737, extentZ=6.9851, center=<0.4544 0 0>}]
 dyn_bracelet_top_base [{shapeType=Capsule, length=0, radius=0.5}]
 dyn_bracelet_top_parent [{shapeType=Box, extentX=1.4284, extentY=6.0203, extentZ=12.3018, center=<0.7527 0 0>}]
}
constraints [
 {bone1Name=dyn_bracelet_bottom_parent, bone2Name=dyn_bracelet_bottom_base, swing1LimitDegrees=10, swing2LimitDegrees=10, twistLimitDegrees=3}
 {bone1Name=dyn_bracelet_top_parent, bone2Name=dyn_bracelet_top_base, swing1LimitDegrees=10, swing2LimitDegrees=10, twistLimitDegrees=3}
]
