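# Physics asset for the necklace of the F_MED_Cavalry female player cosmetic body, letting the choker and its hanging strands move as physics bodies.
# A constrained chain runs from a boxed dyn_choker_base through dyn_choker_parent to dyn_choker_01, then branches into three capsule strands.
# Tight swing and twist limits restrict each joint.
# Larger, unconstrained capsules on neck_01 and spine_05 approximate the neck and upper torso, likely as collision for the necklace.
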PhysicsAsset /BRCosmetics/Characters/Player/Female/Medium/Bodies/F_MED_Cavalry/Meshes/F_MED_Cavalry_Necklace_Physics
bodies {
 dyn_choker_01 [{shapeType=Capsule, length=1.6455, radius=0.7969, shapeName=dyn_choker_01_capsule, center=<1.3983 0 -0.3126>, rotation=<rot 76.1608 0 0>}]
 dyn_choker_base [{shapeType=Box, extentX=0.5, extentY=0.5, extentZ=0.5}]
 dyn_choker_parent [{
   shapeType=Capsule
   length=1.1342
   radius=0.7091
   shapeName=dyn_choker_parent_capsule
   center=<1.0056 0 0.084>
   rotation=<rot 85.2248 180 -180>
  }]
 dyn_choker_strand_01 [{
   shapeType=Capsule
   length=2.9231
   radius=0.735
   shapeName=dyn_choker_strand_01_capsule
   center=<1.9422 0 0.2104>
   rotation=<rot 83.8171 -180 -180>
  }]
 dyn_choker_strand_02 [{
   shapeType=Capsule
   length=3.1065
   radius=0.6981
   shapeName=dyn_choker_strand_02_capsule
   center=<1.9652 0 -0.0638>
   rotation=<rot 89.1169 -180 -180>
  }]
 dyn_strand_choker_03 [{
   shapeType=Capsule
   length=3.2192
   radius=0.6398
   shapeName=dyn_strand_choker_03_capsule
   center=<1.8655 0.0005 -0.0487>
   rotation=<rot 89.395 1.3405 1.3238>
  }]
 neck_01 [{shapeType=Capsule, length=8.625, radius=4.75, center=<-0.0136 -0.0727 0>, rotation=<rot 90 -33.6901 -14.7606>}]
 spine_05 [
  {shapeType=Capsule, length=9.8571, radius=8.5854, shapeName=spine_05_capsule, center=<-1.3842 0.2662 0>}
  {shapeType=Capsule, length=22.875, radius=7.125, shapeName=spine_05_capsule, center=<2.8739 -0.829 0>}
 ]
}
constraints [
 {bone1Name=dyn_choker_parent, bone2Name=dyn_choker_base, swing1LimitDegrees=20, swing2LimitDegrees=10, twistLimitDegrees=2}
 {bone1Name=dyn_choker_01, bone2Name=dyn_choker_parent, swing1LimitDegrees=10, swing2LimitDegrees=10, twistLimitDegrees=2}
 {bone1Name=dyn_choker_strand_01, bone2Name=dyn_choker_01, swing1LimitDegrees=10, swing2LimitDegrees=10, twistLimitDegrees=2}
 {bone1Name=dyn_choker_strand_02, bone2Name=dyn_choker_01, swing1LimitDegrees=10, swing2LimitDegrees=10, twistLimitDegrees=2}
 {bone1Name=dyn_strand_choker_03, bone2Name=dyn_choker_01, swing1LimitDegrees=10, swing2LimitDegrees=10, twistLimitDegrees=2}
]
