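# Physics asset for the F_MED_Cavalry female medium player body's jacket, simulating swinging coat panels and a hanging flask.
# Left and right front and front-side panels, plus a left side panel, are base-parent-01-02 bone chains linked by angle-limited constraints with tight twist.
# The flask chain ends in a more tightly limited liquid box.
# Unconstrained pelvis, thigh and thigh-twist capsules provide body shapes, apparently as colliders for the dangling parts.
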PhysicsAsset /BRCosmetics/Characters/Player/Female/Medium/Bodies/F_MED_Cavalry/Meshes/F_MED_Cavalry_jacket_Physics
bodies {
 dyn_flask_01 [{shapeType=Capsule, length=2.0735, radius=4.4143, center=<4.8114 1.702 -0.0001>, rotation=<rot 90 0 -17.2568>}]
 dyn_flask_base [{shapeType=Sphere, radius=0.5}]
 dyn_flask_liquid [{shapeType=Box, extentX=1.01, extentY=6.26, extentZ=5.51, center=<0.486 0 0>}]
 dyn_flask_parent [{shapeType=Capsule, length=1.0353, radius=1.133, center=<1.7125 0 0>, rotation=<rot 90 0 0>}]
 dyn_jacket_front_l_01 [{
   shapeType=Capsule
   length=4.3657
   radius=0.9833
   shapeName=dyn_jacket_front_l_01_capsule
   center=<0.8686 0 0>
   rotation=<rot -0.0007 0 86.0605>
  }]
 dyn_jacket_front_l_02 [{
   shapeType=Capsule
   length=4.2961
   radius=0.9833
   shapeName=dyn_jacket_front_l_02_capsule
   center=<0.8686 0 0>
   rotation=<rot -0.0007 0 86.0605>
  }]
 dyn_jacket_front_l_base [{shapeType=Sphere, radius=0.5}]
 dyn_jacket_front_l_parent [{
   shapeType=Capsule
   length=4.4594
   radius=1.0733
   shapeName=dyn_jacket_front_l_parent_capsule
   center=<1.1291 0 0>
   rotation=<rot -0.0007 -0.0001 86.0605>
  }]
 dyn_jacket_front_r_01 [{shapeType=Box, extentX=5.1473, extentY=6.4588, extentZ=1.3978, center=<2.5325 0 0>}]
 dyn_jacket_front_r_02 [{shapeType=Box, extentX=2.51, extentY=6.51, extentZ=1.01, center=<1.0933 0 0.3171>}]
 dyn_jacket_front_r_base [{shapeType=Sphere, radius=0.5}]
 dyn_jacket_front_r_parent [{shapeType=Box, extentX=7.0608, extentY=5.568, extentZ=0.75, center=<3.3454 0 0>}]
 dyn_jacket_front_side_l_01 [{
   shapeType=Capsule
   length=4.4011
   radius=1.6399
   shapeName=dyn_jacket_front_side_l_01_capsule
   center=<1.4609 0.02 -0.0035>
   rotation=<rot -0.0001 -0.0001 -68.697>
  }]
 dyn_jacket_front_side_l_02 [{
   shapeType=Capsule
   length=5.282
   radius=1.6366
   shapeName=dyn_jacket_front_side_l_02_capsule
   center=<0.8608 0 0>
   rotation=<rot -0.0001 -0.0001 -67.7806>
  }]
 dyn_jacket_front_side_l_base [{shapeType=Sphere, radius=0.5}]
 dyn_jacket_front_side_l_parent [{
   shapeType=Capsule
   length=5.5932
   radius=1.8145
   shapeName=dyn_jacket_front_side_l_parent_capsule
   center=<1.7328 -0.2334 0.0314>
   rotation=<rot -0.0001 -0.0001 -71.0674>
  }]
 dyn_jacket_front_side_r_01 [{
   shapeType=Capsule
   length=4.7063
   radius=1.4164
   shapeName=dyn_jacket_front_side_r_01_capsule
   center=<1.234 -0.3291 -0.113>
   rotation=<rot 0 0 71.0581>
  }]
 dyn_jacket_front_side_r_02 [{
   shapeType=Capsule
   length=4.3258
   radius=1.4164
   shapeName=dyn_jacket_front_side_r_02_capsule
   center=<1.234 0 0>
   rotation=<rot 0 0 71.0581>
  }]
 dyn_jacket_front_side_r_base [{shapeType=Sphere, radius=0.5}]
 dyn_jacket_front_side_r_parent [{
   shapeType=Capsule
   length=4.8069
   radius=1.4164
   shapeName=dyn_jacket_front_side_r_parent_capsule
   center=<1.234 0 0>
   rotation=<rot 0 0 71.0581>
  }]
 dyn_jacket_side_l_01 [{shapeType=Capsule, length=4.0916, radius=1.3173, shapeName=dyn_jacket_side_l_01_capsule, center=<1.2263 0 0>, rotation=<rot 0 0 -96.1203>}]
 dyn_jacket_side_l_02 [{shapeType=Capsule, length=5.8735, radius=1.1647, shapeName=dyn_jacket_side_l_02_capsule, center=<1.3911 0 0>, rotation=<rot 0 0 -100.2648>}]
 dyn_jacket_side_l_base [{shapeType=Sphere, radius=0.5}]
 dyn_jacket_side_l_parent [{shapeType=Capsule, length=3.9488, radius=1.2507, shapeName=dyn_jacket_side_l_parent_capsule, center=<1.3744 0 0>, rotation=<rot 0 0 85.2315>}]
 pelvis [{shapeType=Capsule, length=14.125, radius=9.5, center=<-0.2217 -1.5993 -0.3567>}]
 thigh_l [{shapeType=Capsule, length=21.375, radius=8.625, center=<-11.6982 -1.1982 0.4286>, rotation=<rot 90 15.1541 19.6857>}]
 thigh_r [{shapeType=Capsule, length=21.875, radius=8.125, center=<14.6314 1.1119 -0.3206>, rotation=<rot 90 0 0>}]
 thigh_twist_01_l [{shapeType=Capsule, length=8.625, radius=8.625, center=<12.3544 0 0>, rotation=<rot 90 0 0>}]
 thigh_twist_01_r [{shapeType=Capsule, length=13.375, radius=8.25, center=<-13.4708 0.8253 -0.0001>, rotation=<rot 90 0 0>}]
}
constraints [
 {bone1Name=dyn_flask_parent, bone2Name=dyn_flask_base, swing1LimitDegrees=20, swing2LimitDegrees=20, twistLimitDegrees=5}
 {bone1Name=dyn_flask_01, bone2Name=dyn_flask_parent, swing1LimitDegrees=20, swing2LimitDegrees=20, twistLimitDegrees=5}
 {bone1Name=dyn_flask_liquid, bone2Name=dyn_flask_01, swing1LimitDegrees=10, swing2LimitDegrees=10, twistLimitDegrees=2}
 {bone1Name=dyn_jacket_front_l_parent, bone2Name=dyn_jacket_front_l_base, swing1LimitDegrees=20, swing2LimitDegrees=40, twistLimitDegrees=3}
 {bone1Name=dyn_jacket_front_l_01, bone2Name=dyn_jacket_front_l_parent, swing1LimitDegrees=20, swing2LimitDegrees=40, twistLimitDegrees=3}
 {bone1Name=dyn_jacket_front_l_02, bone2Name=dyn_jacket_front_l_01, swing1LimitDegrees=20, swing2LimitDegrees=40, twistLimitDegrees=3}
 {
  bone1Name=dyn_jacket_front_side_l_parent
  bone2Name=dyn_jacket_front_side_l_base
  swing1LimitDegrees=20
  swing2LimitDegrees=10
  twistLimitDegrees=3
 }
 {
  bone1Name=dyn_jacket_front_side_l_01
  bone2Name=dyn_jacket_front_side_l_parent
  swing1LimitDegrees=20
  swing2LimitDegrees=10
  twistLimitDegrees=3
 }
 {
  bone1Name=dyn_jacket_front_side_l_02
  bone2Name=dyn_jacket_front_side_l_01
  swing1LimitDegrees=20
  swing2LimitDegrees=10
  twistLimitDegrees=3
 }
 {bone1Name=dyn_jacket_front_r_parent, bone2Name=dyn_jacket_front_r_base, swing1LimitDegrees=20, swing2LimitDegrees=20, twistLimitDegrees=3}
 {bone1Name=dyn_jacket_front_r_01, bone2Name=dyn_jacket_front_r_parent, swing1LimitDegrees=20, swing2LimitDegrees=20, twistLimitDegrees=3}
 {bone1Name=dyn_jacket_front_r_02, bone2Name=dyn_jacket_front_r_01, swing1LimitDegrees=20, swing2LimitDegrees=20, twistLimitDegrees=3}
 {
  bone1Name=dyn_jacket_front_side_r_parent
  bone2Name=dyn_jacket_front_side_r_base
  swing1LimitDegrees=20
  swing2LimitDegrees=20
  twistLimitDegrees=3
 }
 {
  bone1Name=dyn_jacket_front_side_r_01
  bone2Name=dyn_jacket_front_side_r_parent
  swing1LimitDegrees=20
  swing2LimitDegrees=20
  twistLimitDegrees=3
 }
 {
  bone1Name=dyn_jacket_front_side_r_02
  bone2Name=dyn_jacket_front_side_r_01
  swing1LimitDegrees=20
  swing2LimitDegrees=20
  twistLimitDegrees=3
 }
 {bone1Name=dyn_jacket_side_l_parent, bone2Name=dyn_jacket_side_l_base, swing1LimitDegrees=10, swing2LimitDegrees=30, twistLimitDegrees=2}
 {bone1Name=dyn_jacket_side_l_01, bone2Name=dyn_jacket_side_l_parent, swing1LimitDegrees=10, swing2LimitDegrees=30, twistLimitDegrees=2}
 {bone1Name=dyn_jacket_side_l_02, bone2Name=dyn_jacket_side_l_01, swing1LimitDegrees=10, swing2LimitDegrees=30, twistLimitDegrees=2}
]
