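# Post-process style AnimBP for the female Cavalry cosmetic body, adding secondary motion to a pose copied from the parent mesh.
# RigidBody drives jacket and necklace, AnimDynamics moves spikes, belt and bracelets, and a weighted constraint places the choker.
# Thigh rotation relative to pelvis drives additive jacket-bone corrections; stationary crouching and the frontend idle montage each offset the flask.
# The bracelet RigidBody node is left unconnected.
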AnimBlueprint /BRCosmetics/Characters/Player/Female/Medium/Bodies/F_MED_Cavalry/Meshes/F_MED_Cavalry_AnimBP
target_skeleton=&/F_MED_Cavalry_Skeleton
graphs [
 AnimGraph {nodes [
   RigidBody() {
    bClampLinearTranslationLimitToRefPose=true
    bOverrideWorldGravity=true
    overridePhysicsAsset=&/F_MED_Cavalry_Bracelets_Physics
    simulationSpace=BaseBoneSpace
    alphaBoolBlend {blendOption=Linear}
    baseBoneRef {boneName=spine_05}
    componentAppliedLinearAccClamp=<3000 3000 3000>
    componentLinearAccScale=<0.25 0.25 0.25>
    componentLinearVelScale=<0.25 0.25 0.25>
    overrideWorldGravity {z=-980}
   }
   n1 = RigidBody(ComponentPose=Local To Component(LocalPose=Copy Pose From Mesh() {bUseAttachedParent=true})) {
    bClampLinearTranslationLimitToRefPose=true
    bOverrideWorldGravity=true
    overridePhysicsAsset=&/F_MED_Cavalry_jacket_Physics
    simulationSpace=BaseBoneSpace
    alphaBoolBlend {blendOption=Linear}
    alphaScaleBiasClamp {bInterpResult=true, interpSpeedDecreasing=6, interpSpeedIncreasing=8}
    baseBoneRef {boneName=pelvis}
    componentAppliedLinearAccClamp=<3000 3000 3000>
    componentLinearAccScale=<0.25 0.25 0.5>
    componentLinearVelScale=<0.25 0.25 0.5>
    overrideWorldGravity {z=-980}
   }
   n2 = Anim Dynamics - Bone: dyn_spike_l_base(ComponentPose=@n1, Alpha=1, ExternalForce="0,0,0") {
    angularDampingOverride=0.99
    angularSpringConstant=50
    bAngularSpring=true
    bGravityOverrideInSimSpace=true
    bOverrideAngularDamping=true
    simulationSpace=RootRelative
    alphaBoolBlend {blendOption=Linear}
    boundBone {boneName=dyn_spike_l_base}
    componentAppliedLinearAccClamp=<3500 3500 3500>
    componentLinearAccScale=<0.8 0.8 0.8>
    componentLinearVelScale=<0.8 0.8 0.8>
    gravityOverride {z=-810}
    physicsBodyDefinitions [{
      sphereCollisionRadius=0
      boundBone {boneName=dyn_spike_l_base}
      boxExtents=<15 15 10>
      constraintSetup {angularLimitsMax=<5 15 5>, angularLimitsMin=<-5 0 -5>, angularTarget=<0 0 0>}
      localJointOffset=<3 0 0>
     }]
    relativeSpaceBone {boneName=root}
   }
   n3 = Anim Dynamics - Bone: dyn_spike_r_base(ComponentPose=@n2, Alpha=1, ExternalForce="0,0,0") {
    angularDampingOverride=0.99
    angularSpringConstant=50
    bAngularSpring=true
    bGravityOverrideInSimSpace=true
    bOverrideAngularDamping=true
    simulationSpace=RootRelative
    alphaBoolBlend {blendOption=Linear}
    boundBone {boneName=dyn_spike_r_base}
    componentAppliedLinearAccClamp=<3500 3500 3500>
    componentLinearAccScale=<0.8 0.8 0.8>
    componentLinearVelScale=<0.8 0.8 0.8>
    gravityOverride {z=-810}
    physicsBodyDefinitions [{
      sphereCollisionRadius=0
      boundBone {boneName=dyn_spike_r_base}
      boxExtents=<15 15 10>
      constraintSetup {angularLimitsMax=<5 15 5>, angularLimitsMin=<-5 0 -5>, angularTarget=<0 0 0>}
      localJointOffset=<3 0 0>
     }]
    relativeSpaceBone {boneName=root}
   }
   n4 = Anim Dynamics - Bone: dyn_belt_base(ComponentPose=@n3, Alpha=0.8, ExternalForce="0,0,0") {
    angularDampingOverride=0.98
    angularSpringConstant=500
    bAngularSpring=true
    bGravityOverrideInSimSpace=true
    bOverrideAngularDamping=true
    bUsePlanarLimit=false
    simulationSpace=RootRelative
    alphaBoolBlend {blendOption=Linear}
    boundBone {boneName=dyn_belt_base}
    componentAppliedLinearAccClamp=<3500 3500 3500>
    componentLinearAccScale=<0.9 0.9 0.9>
    componentLinearVelScale=<0.9 0.9 0.9>
    physicsBodyDefinitions [{
      sphereCollisionRadius=4
      boundBone {boneName=dyn_belt_base}
      boxExtents=<25 15 15>
      constraintSetup {angularLimitsMax=<1 15 5>, angularLimitsMin=<-1 -5 -5>, angularTarget=<1 0 0>}
      localJointOffset=<2 0 0>
     }]
    relativeSpaceBone {boneName=pelvis}
    sphericalLimits [{limitRadius=6, drivingBone {boneName=upperarm_twist_02_r}, sphereLocalOffset=<0 0 3>}]
   }
   n5 = Anim Dynamics - Bone: dyn_bracelet_top_base(ComponentPose=@n4, Alpha=1, ExternalForce="0,0,0") {
    angularDampingOverride=0.95
    angularSpringConstant=1500
    bAngularSpring=true
    bGravityOverrideInSimSpace=true
    bOverrideAngularDamping=true
    gravityScale=0.75
    simulationSpace=RootRelative
    alphaBoolBlend {blendOption=Linear}
    boundBone {boneName=dyn_bracelet_top_base}
    componentAppliedLinearAccClamp=<500 500 500>
    componentLinearAccScale=<1 1 1>
    componentLinearVelScale=<1 1 1>
    physicsBodyDefinitions [{
      sphereCollisionRadius=0
      boundBone {boneName=dyn_bracelet_top_base}
      boxExtents=<20 20 20>
      constraintSetup {angularLimitsMax=<30 15 0>, angularLimitsMin=<-30 -15 -10>, angularTarget=<1 0 0>, linearAxesMax=<0.5 0 0>}
      localJointOffset=<0.5 0 0.5>
     }]
    relativeSpaceBone {boneName=pelvis}
   }
   n6 = Anim Dynamics - Bone: dyn_bracelet_bottom_base(ComponentPose=@n5, Alpha=0.8, ExternalForce="0,0,0") {
    angularDampingOverride=0.95
    angularSpringConstant=1500
    bAngularSpring=true
    bGravityOverrideInSimSpace=true
    bOverrideAngularDamping=true
    gravityScale=0.75
    simulationSpace=RootRelative
    alphaBoolBlend {blendOption=Linear}
    boundBone {boneName=dyn_bracelet_bottom_base}
    componentAppliedLinearAccClamp=<500 500 500>
    componentLinearAccScale=<0.8 0.8 0.8>
    componentLinearVelScale=<0.8 0.8 0.8>
    physicsBodyDefinitions [{
      sphereCollisionRadius=0
      boundBone {boneName=dyn_bracelet_bottom_base}
      boxExtents=<20 20 20>
      constraintSetup {angularLimitsMax=<5 5 5>, angularLimitsMin=<-5 -5 -5>, angularTarget=<1 0 0>, linearAxesMax=<0.5 0 0>}
      localJointOffset=<0.5 0 0.5>
     }]
    relativeSpaceBone {boneName=pelvis}
   }
   n7 = Get Thigh_Rot_R() -> (Thigh_Rot_R:double)
   n8 = "Transform (Modify) Bone - Bone: dyn_jacket_front_side_r_base"(ComponentPose=@n6, Alpha=@n7, Translation="0,0,0", Rotation="-85,-100,0", Scale="1,1,1") {
    rotationMode=BMM_Additive
    translationMode=BMM_Additive
    alphaBoolBlend {blendOption=Linear}
    boneToModify {boneName=dyn_jacket_front_side_r_base}
    rotation {pitch=-85, yaw=-100}
   }
   n9 = "Transform (Modify) Bone - Bone: dyn_jacket_front_r_base"(ComponentPose=@n8, Alpha=@n7, Translation="0,0,0", Rotation="-40,-40,0", Scale="1,1,1") {
    rotationMode=BMM_Additive
    translationMode=BMM_Additive
    alphaBoolBlend {blendOption=Linear}
    boneToModify {boneName=dyn_jacket_front_r_base}
    rotation {pitch=-40, yaw=-40}
   }
   n10 = Get Thigh_Rot_L() -> (Thigh_Rot_L:double)
   n11 = "Transform (Modify) Bone - Bone: dyn_jacket_front_l_base"(ComponentPose=@n9, Alpha=@n10, Translation="0,0,0", Rotation="-30,0,0", Scale="1,1,1") {
    rotationMode=BMM_Additive
    rotationSpace=BCS_BoneSpace
    translationMode=BMM_Additive
    translationSpace=BCS_BoneSpace
    alphaBoolBlend {blendOption=Linear}
    alphaScaleBiasClamp {bInterpResult=true}
    boneToModify {boneName=dyn_jacket_front_l_base}
    rotation {pitch=-30}
   }
   n12 = "Transform (Modify) Bone - Bone: dyn_jacket_front_side_l_base"(ComponentPose=@n11, Alpha=@n10, Translation="0,0,0", Rotation="-110,10,0", Scale="1,1,1") {
    rotationMode=BMM_Additive
    rotationSpace=BCS_BoneSpace
    translationMode=BMM_Additive
    translationSpace=BCS_BoneSpace
    alphaBoolBlend {blendOption=Linear}
    alphaScaleBiasClamp {bInterpResult=true}
    boneToModify {boneName=dyn_jacket_front_side_l_base}
    rotation {pitch=-110, yaw=10}
   }
   n13 = "Transform (Modify) Bone - Bone: dyn_jacket_side_l_base"(ComponentPose=@n12, Alpha=@n10, Translation="0,0,0", Rotation="-120,0,0", Scale="1,1,1") {
    rotationMode=BMM_Additive
    rotationSpace=BCS_BoneSpace
    translationMode=BMM_Additive
    translationSpace=BCS_BoneSpace
    alphaBoolBlend {blendOption=Linear}
    alphaScaleBiasClamp {bInterpResult=true}
    boneToModify {boneName=dyn_jacket_side_l_base}
    rotation {pitch=-120}
   }
   n14 = "Transform (Modify) Bone - Bone: dyn_flask_base"(ComponentPose=@n13, Alpha=Get Crouch_Alpha() -> (Crouch_Alpha:double), Translation="0,1.5,0", Rotation="0,35,0", Scale="1,1,1") {
    rotationMode=BMM_Additive
    rotationSpace=BCS_BoneSpace
    translationMode=BMM_Additive
    translationSpace=BCS_BoneSpace
    alphaBoolBlend {blendOption=Linear}
    alphaScaleBiasClamp {bInterpResult=true, interpSpeedDecreasing=6, interpSpeedIncreasing=8}
    boneToModify {boneName=dyn_flask_base}
    rotation {yaw=35}
    translation {y=1.5}
   }
   n15 = "Constraint - dyn_choker_base (PPPPPPP)"!(
    ComponentPose=@n14
    Alpha=1
    ConstraintWeights_0=0.1711
    ConstraintWeights_1:float=0.0012
    ConstraintWeights_2:float=0.0015
    ConstraintWeights_3:float=0.5125
    ConstraintWeights_4:float=0.0015
    ConstraintWeights_5:float=0.1268
    ConstraintWeights_6:float=0.1854
   ) -> (Pose) {
    alphaBoolBlend {blendOption=Linear}
    boneToModify {boneName=dyn_choker_base}
    constraintSetup [
     {targetBone {boneName=head}}
     {targetBone {boneName=clavicle_l}}
     {targetBone {boneName=spine_04}}
     {targetBone {boneName=spine_05}}
     {targetBone {boneName=clavicle_r}}
     {targetBone {boneName=neck_02}}
     {targetBone {boneName=neck_01}}
    ]
    constraintWeights [1,1,1,1,1,1,1]
   }
   n16 = RigidBody(ComponentPose=@n15) {
    bClampLinearTranslationLimitToRefPose=true
    bOverrideWorldGravity=true
    overridePhysicsAsset=&/F_MED_Cavalry_Necklace_Physics
    simulationSpace=BaseBoneSpace
    alphaBoolBlend {blendOption=Linear}
    baseBoneRef {boneName=spine_02}
    componentAppliedLinearAccClamp=<3000 3000 3000>
    componentLinearAccScale=<0.25 0.25 0.25>
    componentLinearVelScale=<0.25 0.25 0.25>
    overrideWorldGravity {z=-980}
   }
   n17 = "Transform (Modify) Bone - Bone: dyn_flask_base"!(
    ComponentPose=@n16
    bAlphaBoolEnabled:bool=Get IsFrontendIdle() -> (IsFrontendIdle:bool)
    Translation="0,0,0"
    Rotation="12,0,0"
    Scale="1,1,1"
   ) -> (Pose) {
    alphaInputType=Bool
    rotationMode=BMM_Additive
    rotationSpace=BCS_BoneSpace
    translationMode=BMM_Additive
    translationSpace=BCS_BoneSpace
    alphaBoolBlend {blendOption=Linear}
    boneToModify {boneName=dyn_flask_base}
    rotation {pitch=12}
   }
   Output Pose(Result=Component To Local(ComponentPose=@n17)) {layerGroup=None}
  ]}
 EventGraph {nodes [
   Event BlueprintUpdateAnimation()
   > Get!(execute:Exec) -> (then:Exec, else:Exec, Skel_Mesh:Skeletal Mesh Component Object Reference)
   > n1 = Sequence!(execute) -> (then_0, then_1, then_2:Exec, then_3:Exec)
   n2 = "float <= float"!(A:double=Get PawnSpeed2D(), B:double=1) -> (ReturnValue:bool) {is=title K2Node_CallFunction}
   n3 = Get bIsCrouching()
   > Set Crouch_Alpha(Crouch_Alpha:double=SelectFloat(A=1, B=0, bPickA=AND Boolean(A=@n3, B=@n2))) -> (Output_Get:double)
   Set Thigh_Rot_L(
    execute=@n1.then_1
    Thigh_Rot_L:double=GetAlphaFromReferenceTransformDelta(SkeletalMeshComponent=Get Skel_Mesh(), BoneName=thigh_l, BaseName=pelvis, InputRange=(X=-19,Y=-101), TwistAxis="0,0,1")
   ) -> (Output_Get:double)
   Set Thigh_Rot_R(
    execute=@n1.then_2
    Thigh_Rot_R:double=GetAlphaFromReferenceTransformDelta(SkeletalMeshComponent=Get Skel_Mesh(), BoneName=thigh_r, BaseName=pelvis, InputRange=(X=4,Y=-101), TwistAxis="0,0,1")
   ) -> (Output_Get:double)
   n4 = GetAnimInstance(self=GetOwnerSkeletalMeshComponent())
   Is Valid(exec=@n1.then_3, InputObject=@n4)
   > Set IsFrontendIdle(IsFrontendIdle:bool=Montage_IsActive!(
     self:Anim Instance Object Reference=@n4
     Montage:Anim Montage Object Reference=/BRCosmetics/Animation/Game/MainPlayer/Menu/BR/CustomSkins/_S14/Cavalry/Cavalry_Female_Idle_Montage
    ) -> (ReturnValue:bool)) -> (Output_Get:bool)
   Event BlueprintInitializeAnimation()
   > Set Skel_Mesh(Skel_Mesh=GetOwnerSkeletalMeshComponent())
  ]}
]
variables [
 Crouch_Alpha {type=number}
 IsFrontendIdle {type=boolean}
 Skel_Mesh {type=SkeletalMeshComponent}
 Thigh_Rot_L {type=number}
 Thigh_Rot_R {type=number}
]
