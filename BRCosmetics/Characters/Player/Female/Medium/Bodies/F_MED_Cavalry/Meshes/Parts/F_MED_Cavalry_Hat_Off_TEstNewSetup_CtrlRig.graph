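# Control Rig for the hat-off F_MED_Cavalry cosmetic part, apparently a test setup, that poses hair bones to follow and avoid the body.
# Constraints_FNC parent-constrains hair between head, neck and upper spine, then blends offsets by head yaw relative to spine_05.
# HairAvoidance_fnc then layers bone offsets weighted by spherical pose readers on both upper arms.
# Skydive, targeting and pistol flags enable or suppress the left-arm offsets.
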ControlRigBlueprint /BRCosmetics/Characters/Player/Female/Medium/Bodies/F_MED_Cavalry/Meshes/Parts/F_MED_Cavalry_Hat_Off_TEstNewSetup_CtrlRig
graphs [
 Multiply_2_ContainedGraph {nodes [
   n1 = RigVMFunctionEntryNode()
   Multiply = &RigVMFunction_MathDoubleMul(A=@n1.A, B=@n1.B)
   RigVMFunctionReturnNode(Result=&RigVMFunction_MathDoubleMul(B=@n1.C, A=@Multiply.Result).Result)
  ]}
 New Function_ContainedGraph {nodes [
   RigVMFunctionEntryNode()
   > &RigUnit_ParentConstraint(
    AdvancedSettings=(InterpolationType=Average,RotationOrderForFilter=XZY)
    Child=(Type=Bone,Name="hairMaster")
    Parents=((Item=(Type=Bone,Name="spine_05"),Weight=0.2),(Item=(Type=Bone,Name="neck_02"),Weight=1),(Item=(Type=Bone,Name="neck_01"),Weight=0))
    Weight=1
   )
   > &RigUnit_ParentConstraint(
    AdvancedSettings=(InterpolationType=Average,RotationOrderForFilter=XZY)
    Child=(Type=Bone,Name="dyn_hair_front_r_01")
    Parents=((Item=(Type=Bone,Name="head"),Weight=0.6),(Item=(Type=Bone,Name="spine_05"),Weight=0.6))
    Weight=1
   )
   > &RigUnit_ParentConstraint(
    AdvancedSettings=(InterpolationType=Average,RotationOrderForFilter=XZY)
    Child=(Type=Bone,Name="dyn_hair_side_r_01")
    Parents=((Item=(Type=Bone,Name="head"),Weight=0.6),(Item=(Type=Bone,Name="spine_05"),Weight=0.6))
    Weight=1
   )
   > &RigUnit_ParentConstraint(
    AdvancedSettings=(InterpolationType=Average,RotationOrderForFilter=XZY)
    Child=(Type=Bone,Name="dyn_hair_side_l_02")
    Parents=((Item=(Type=Bone,Name="head"),Weight=0.4),(Item=(Type=Bone,Name="spine_05"),Weight=0.6))
    Weight=1
   )
   > &RigUnit_ParentConstraint(
    AdvancedSettings=(InterpolationType=Average,RotationOrderForFilter=XZY)
    Child=(Type=Bone,Name="dyn_hair_front_l_01")
    Parents=((Item=(Type=Bone,Name="head"),Weight=0.3),(Item=(Type=Bone,Name="spine_05"),Weight=0.5))
    Weight=1
   )
   n1 = /Script/RigVM.RigVMFunction_MathQuaternionToRotator(Value=&RigVMFunction_MathTransformMakeRelative(
     Global=&RigUnit_GetRelativeTransformForItem(Child=(Type=Bone,Name="head"), Parent=(Type=Bone,Name="head"), bChildInitial=False, bParentInitial=true).RelativeTransform
     Parent=&RigUnit_GetRelativeTransformForItem(Child=(Type=Bone,Name="spine_05"), Parent=(Type=Bone,Name="spine_05"), bChildInitial=False, bParentInitial=true).RelativeTransform
    ).Local.Rotation) {is=struct RigVMUnitNode, template="Cast::Execute(in Value,out Result)"}
   > &RigUnit_ModifyTransforms(
    ItemToModify="((Item=(Type=Bone,Name=\"hairMaster\"),Transform=(Rotation=(X=0.1053,Y=-0.0049,Z=-0.0466,W=-0.9933),Translation=(X=2.0473,Y=-5.2615,Z=1.1286),Scale3D=(X=1,Y=1,Z=1))),(Item=(Type=Bone,Name=\"base_backHair_R_01\"),Transform=(Rotation=(X=0.0796,Y=0.0471,Z=-0.0034 ...(truncated)"
    WeightMinimum=0
    Weight=&RigVMFunction_MathDoubleRemap(
     SourceMaximum=-55
     SourceMinimum=0
     TargetMaximum=1
     TargetMinimum=0
     Value=RigVMDispatchNode(Struct=@n1.Result) {template="DISPATCH_RigVMDispatch_BreakStruct(in Struct,out Elements)"}.Elements.Yaw
    ).Result
   )
   > RigVMFunctionReturnNode()
  ]}
 New Function_ContainedGraph {nodes [
   RigVMFunctionEntryNode()
   > n1 = &RigUnit_SphericalPoseReader(
    ActiveRegionScaleFactors=(PositiveWidth=3,NegativeWidth=3,PositiveHeight=3,NegativeHeight=3)
    ActiveRegionSize=0.1
    Debug=(bDrawDebug=false,bDraw2D=False,bDrawLocalAxes=False,DebugScale=25,DebugSegments=20,DebugThickness=0.25)
    DriverAxis=(X=1,Y=0,Z=0)
    DriverItem=(Type=Bone,Name="upperarm_l")
    FalloffRegionScaleFactors=(PositiveWidth=3,NegativeWidth=3,PositiveHeight=3,NegativeHeight=3)
    FalloffSize=0.4
    FlipHeightScaling=False
    FlipWidthScaling=False
    OptionalParentItem=(Type=None,Name="None")
    RotationOffset=(X=-15,Y=0,Z=0)
   )
   RigVMFunction_MathBoolToFloat = &RigVMFunction_MathBoolToFloat(Value=RigVMVariableNode() {variable=isSkydive}.Value)
   AlphaInterp = &RigVMFunction_AlphaInterp(
    InterpSpeedDecreasing=10
    InterpSpeedIncreasing=10
    Scale=1
    bClampResult=true
    bInterpResult=False
    Value=@RigVMFunction_MathBoolToFloat.Result
   )
   > &RigUnit_ModifyTransforms(
    ItemToModify="((Item=(Type=Bone,Name=\"base_backHair_R_01\"),Transform=(Rotation=(X=-0.0074,Y=0.0444,Z=-0.1633,W=-0.9856),Translation=(X=0,Y=-3.6958,Z=0),Scale3D=(X=1,Y=1,Z=1))),(Item=(Type=Bone,Name=\"base_backHair_L_01\"),Transform=(Rotation=(X=0,Y=0,Z=-0.1233,W=-0.9924), ...(truncated)"
    WeightMinimum=0
    Weight=&RigVMFunction_MathDoubleMul(B=@AlphaInterp.Result, A=@n1.OutputParam).Result
   )
   > n2 = &RigUnit_SphericalPoseReader(
    ActiveRegionScaleFactors=(PositiveWidth=1.5,NegativeWidth=1.5,PositiveHeight=0.9,NegativeHeight=1.3)
    ActiveRegionSize=0.1
    Debug=(bDrawDebug=false,DebugScale=25,DebugSegments=20,DebugThickness=0.25)
    DriverAxis=(X=1,Y=0,Z=0)
    DriverItem=(Type=Bone,Name="upperarm_l")
    FalloffRegionScaleFactors=(PositiveWidth=0.5,NegativeWidth=0.5,PositiveHeight=0.3,NegativeHeight=0.5)
    FalloffSize=0.5
    OptionalParentItem=(Type=Bone)
    RotationOffset=(X=90,Y=50,Z=-50)
   )
   n3 = &RigVMFunction_AlphaInterp(
    InterpSpeedDecreasing=10
    InterpSpeedIncreasing=10
    OutRange=(Min=1,Max=0)
    Scale=1
    bClampResult=true
    bMapRange=true
    Value=&RigVMFunction_MathBoolToFloat(Value=RigVMVariableNode() {variable=isTargeting}.Value).Result
   )
   > &RigUnit_ModifyTransforms(
    ItemToModify="((Item=(Type=Bone,Name=\"base_backHair_R_01\"),Transform=(Rotation=(X=-0.3213,Y=0.1919,Z=-0.0215,W=0.9271),Translation=(X=2,Y=0,Z=3),Scale3D=(X=1,Y=1,Z=1))),(Item=(Type=Bone,Name=\"dyn_backHair_R_02\"),Transform=(Rotation=(X=0,Y=0,Z=0,W=1),Translation=(X=0,Y=0 ...(truncated)"
    WeightMinimum=0
    Weight=&RigVMFunction_MathDoubleMul(A=@n3.Result, B=@n2.OutputParam).Result
   )
   > n4 = &RigUnit_SphericalPoseReader(
    ActiveRegionScaleFactors=(PositiveWidth=2.5,NegativeWidth=3,PositiveHeight=3,NegativeHeight=3)
    ActiveRegionSize=0.1
    Debug=(bDrawDebug=false,DebugScale=25,DebugSegments=20,DebugThickness=0.25)
    DriverAxis=(X=1,Y=0,Z=0)
    DriverItem=(Type=Bone,Name="upperarm_l")
    FalloffRegionScaleFactors=(PositiveWidth=1,NegativeWidth=1,PositiveHeight=1,NegativeHeight=1)
    FalloffSize=0.5
    OptionalParentItem=(Type=Bone)
    RotationOffset=(X=0,Y=0,Z=0)
   )
   n5 = &RigVMFunction_AlphaInterp(
    InterpSpeedDecreasing=10
    InterpSpeedIncreasing=10
    OutRange=(Min=1,Max=0)
    Scale=1
    bClampResult=true
    bMapRange=true
    Value=&RigVMFunction_MathBoolToFloat(Value=RigVMVariableNode() {variable=GetPistol}.Value).Result
   )
   n6 = &RigVMFunction_AlphaInterp(
    InterpSpeedDecreasing=10
    InterpSpeedIncreasing=10
    OutRange=(Min=1,Max=0)
    Scale=1
    bClampResult=true
    bMapRange=true
    Value=&RigVMFunction_MathBoolToFloat(Value=RigVMVariableNode() {variable=isSkydive}.Value).Result
   )
   Multiply = RigVMAggregateNode(B=@n5.Result, C=@n6.Result, A=@n4.OutputParam)
   > &RigUnit_ModifyTransforms(
    ItemToModify="((Item=(Type=Bone,Name=\"hairMaster\"),Transform=(Rotation=(X=-0.217,Y=0.0021,Z=0.0096,W=0.9761),Translation=(X=0.7651,Y=-1.7773,Z=-0.5645),Scale3D=(X=1,Y=1,Z=1))),(Item=(Type=Bone,Name=\"base_backHair_R_01\"),Transform=(Rotation=(X=0.3305,Y=-0.039,Z=-0.0272,W ...(truncated)"
    WeightMinimum=0
    Weight=@Multiply.Result
   )
   > n7 = &RigUnit_SphericalPoseReader(
    ActiveRegionScaleFactors=(PositiveWidth=1,NegativeWidth=1,PositiveHeight=1,NegativeHeight=0.8)
    ActiveRegionSize=0.1
    Debug=(bDrawDebug=false,bDraw2D=False,bDrawLocalAxes=False,DebugScale=25,DebugSegments=20,DebugThickness=0.25)
    DriverAxis=(X=1,Y=0,Z=0)
    DriverItem=(Type=Bone,Name="upperarm_r")
    FalloffRegionScaleFactors=(PositiveWidth=1,NegativeWidth=1,PositiveHeight=1.7,NegativeHeight=0.5)
    FalloffSize=0.2
    FlipHeightScaling=False
    FlipWidthScaling=False
    OptionalParentItem=(Type=Bone,Name="None")
    RotationOffset=(X=51.9395,Y=8.7023,Z=22.4256)
   )
   > &RigUnit_ModifyTransforms(
    ItemToModify="((Item=(Type=Bone,Name=\"hairMaster\"),Transform=(Rotation=(X=-0.1016,Y=0.0017,Z=0.0164,W=0.9947),Translation=(X=0.7994,Y=-1.228,Z=-0.2886),Scale3D=(X=1,Y=1,Z=1))),(Item=(Type=Bone,Name=\"base_backHair_L_01\"),Transform=(Rotation=(X=0.3728,Y=-0.0475,Z=0.1171,W ...(truncated)"
    WeightMinimum=0
    Weight=@n7.OutputParam
   )
   > &RigUnit_ModifyTransforms(
    ItemToModify="((Item=(Type=Bone,Name=\"dyn_hair_side_r_base\"),Transform=(Rotation=(X=0,Y=-0.1288,Z=0,W=0.9917),Translation=(X=-0.2816,Y=0,Z=1.0659),Scale3D=(X=1,Y=1,Z=1))),(Item=(Type=Bone,Name=\"dyn_hair_front_r_base\"),Transform=(Rotation=(X=0,Y=0.121,Z=0,W=-0.9927),Tran ...(truncated)"
    WeightMinimum=0
    Weight=@n7.OutputParam
   )
   > RigVMFunctionReturnNode()
  ]}
 RigVMFunctionLibrary {nodes [
   Constraints_FNC = RigVMCollapseNode(ExecuteContext=())
   HairAvoidance_fnc = RigVMCollapseNode(ExecuteContext=())
  ]}
 RigVMModel {nodes [
   RigUnit_BeginExecution = &RigUnit_BeginExecution()
   Constraints_FNC = RigVMFunctionReferenceNode(ExecuteContext=@RigUnit_BeginExecution.ExecutePin)
   > HairAvoidance_fnc = RigVMFunctionReferenceNode()
  ]}
 RigVMModel Constraints_FNC {nodes [
   &RigUnit_ParentConstraint(
    AdvancedSettings=(InterpolationType=Average,RotationOrderForFilter=XZY)
    Child=(Type=Bone,Name="hairMaster")
    ExecutePin=()
    Parents=((Item=(Type=Bone,Name="spine_05"),Weight=0.2),(Item=(Type=Bone,Name="neck_02"),Weight=1),(Item=(Type=Bone,Name="neck_01"),Weight=0))
    Weight=1
   )
   > &RigUnit_ParentConstraint(
    AdvancedSettings=(InterpolationType=Average,RotationOrderForFilter=XZY)
    Child=(Type=Bone,Name="dyn_hair_front_r_01")
    Parents=((Item=(Type=Bone,Name="head"),Weight=0.6),(Item=(Type=Bone,Name="spine_05"),Weight=0.6))
    Weight=1
   )
   > &RigUnit_ParentConstraint(
    AdvancedSettings=(InterpolationType=Average,RotationOrderForFilter=XZY)
    Child=(Type=Bone,Name="dyn_hair_side_r_01")
    Parents=((Item=(Type=Bone,Name="head"),Weight=0.6),(Item=(Type=Bone,Name="spine_05"),Weight=0.6))
    Weight=1
   )
   > &RigUnit_ParentConstraint(
    AdvancedSettings=(InterpolationType=Average,RotationOrderForFilter=XZY)
    Child=(Type=Bone,Name="dyn_hair_side_l_02")
    Parents=((Item=(Type=Bone,Name="head"),Weight=0.4),(Item=(Type=Bone,Name="spine_05"),Weight=0.6))
    Weight=1
   )
   > &RigUnit_ParentConstraint(
    AdvancedSettings=(InterpolationType=Average,RotationOrderForFilter=XZY)
    Child=(Type=Bone,Name="dyn_hair_front_l_01")
    Parents=((Item=(Type=Bone,Name="head"),Weight=0.3),(Item=(Type=Bone,Name="spine_05"),Weight=0.5))
    Weight=1
   )
   n1 = &RigVMFunction_MathQuaternionToRotator(Value=&RigVMFunction_MathTransformMakeRelative(
     Global=&RigUnit_GetRelativeTransformForItem(Child=(Type=Bone,Name="head"), Parent=(Type=Bone,Name="head"), bChildInitial=False, bParentInitial=true).RelativeTransform
     Parent=&RigUnit_GetRelativeTransformForItem(Child=(Type=Bone,Name="spine_05"), Parent=(Type=Bone,Name="spine_05"), bChildInitial=False, bParentInitial=true).RelativeTransform
    ).Local.Rotation) {template="Cast::Execute(in Value,out Result)"}
   > &RigUnit_ModifyTransforms(
    ItemToModify="((Item=(Type=Bone,Name=\"hairMaster\"),Transform=(Rotation=(X=0.1053,Y=-0.0049,Z=-0.0466,W=-0.9933),Translation=(X=2.0473,Y=-5.2615,Z=1.1286),Scale3D=(X=1,Y=1,Z=1))),(Item=(Type=Bone,Name=\"base_backHair_R_01\"),Transform=(Rotation=(X=0.0796,Y=0.0471,Z=-0.0034 ...(truncated)"
    WeightMinimum=0
    Weight=&RigVMFunction_MathDoubleRemap(
     SourceMaximum=-55
     SourceMinimum=0
     TargetMaximum=1
     TargetMinimum=0
     Value=RigVMDispatchNode(Struct=@n1.Result) {template="DISPATCH_RigVMDispatch_BreakStruct(in Struct,out Elements)"}.Elements.Yaw
    ).Result
   )
  ]}
]
hierarchy [
 root {loc=<0 0 0>, rot=<0 0 0 1>}
 pelvis {parent=root, loc=<0 -0.5301 98.0571>, rot=<0.0127 0.707 -0.0127 -0.707>}
 spine_01 {parent=pelvis, loc=<6.1208 -0.9328 0>, rot=<0 0 0.034 -0.9994>}
 spine_02 {parent=spine_01, loc=<6.5996 -0.2095 0>, rot=<0 0 0.0238 -0.9997>}
 spine_03 {parent=spine_02, loc=<6.4914 -0.1237 0>, rot=<0 0 0.0149 -0.9999>}
 spine_04 {parent=spine_03, loc=<6.9598 0.0794 0>, rot=<0 0 -0.0132 -0.9999>}
 spine_05 {parent=spine_04, loc=<6.9987 0 0>, rot=<0 0 -0.0015 -1>}
 clavicle_l {parent=spine_05, loc=<5.4831 -0.1053 -3.2946>, rot=<0.0465 -0.7289 -0.009 -0.6829>}
 upperarm_l {parent=clavicle_l, loc=<9.8658 -1.2136 -0.4105>, rot=<0.0208 -0.486 -0.0242 -0.8734>}
 lowerarm_l {parent=upperarm_l, loc=<26.465 -0.8345 0.259>, rot=<0.0036 0.0161 -0.2298 -0.9731>}
 hand_l {parent=lowerarm_l, loc=<22.7355 -0.3203 -0.6105>, rot=<-0.6784 -0.0223 0.0735 -0.7307>}
 index_metacarpal_l {parent=hand_l, loc=<1.6535 0.0183 -2.96>, rot=<0.0158 -0.0124 -0.0173 -0.9996>}
 index_01_l {parent=index_metacarpal_l, loc=<8.3008 0.0357 -0.0524>, rot=<-0.0044 0.0485 0.0924 -0.9945>}
 index_02_l {parent=index_01_l, loc=<3.3587 -0.3004 0.0452>, rot=<-0.002 0.0027 0.3225 -0.9466>}
 index_03_l {parent=index_02_l, loc=<2.5578 -0.123 -0.0252>, rot=<-0.002 0.0067 0.2558 -0.9667>}
 middle_metacarpal_l {parent=hand_l, loc=<2.1856 -0.1725 -0.5984>, rot=<0.006 0.0168 -0.0329 -0.9993>}
 middle_01_l {parent=middle_metacarpal_l, loc=<7.3275 -0.2447 0.0129>, rot=<-0.0208 0.0042 0.0858 -0.9961>}
 middle_02_l {parent=middle_01_l, loc=<4.0123 0.1401 0.1175>, rot=<-0.0218 0.0258 0.3512 -0.9357>}
 middle_03_l {parent=middle_02_l, loc=<2.6077 0.0534 0.1373>, rot=<0.0082 0.0188 0.1708 -0.9851>}
 pinky_metacarpal_l {parent=hand_l, loc=<1.7616 -0.9753 3.3814>, rot=<-0.0397 0.0814 -0.069 -0.9935>}
 pinky_01_l {parent=pinky_metacarpal_l, loc=<7.1709 -0.4831 -0.1094>, rot=<-0.0654 -0.0035 0.2242 -0.9723>}
 pinky_02_l {parent=pinky_01_l, loc=<2.8274 0.0893 -0.1043>, rot=<0.0037 0.0011 0.2854 -0.9584>}
 pinky_03_l {parent=pinky_02_l, loc=<2.2046 0.0307 -0.0078>, rot=<-0.0002 0.0098 0.1445 -0.9895>}
 ring_metacarpal_l {parent=hand_l, loc=<2.2074 -0.3625 1.3638>, rot=<-0.0284 0.0722 -0.0458 -0.9959>}
 ring_01_l {parent=ring_metacarpal_l, loc=<7.0295 -0.6197 -0.1446>, rot=<-0.0489 -0.0252 0.1892 -0.9804>}
 ring_02_l {parent=ring_01_l, loc=<3.6174 0.4647 -0.1027>, rot=<0 0 0.2806 -0.9598>}
 ring_03_l {parent=ring_02_l, loc=<2.4152 0.114 0.0199>, rot=<0 0 0.1189 -0.9929>}
 thumb_01_l {parent=hand_l, loc=<1.3118 -0.3191 -2.459>, rot=<0.6546 -0.4042 0.0664 -0.6354>}
 thumb_02_l {parent=thumb_01_l, loc=<5.9336 -0.0448 0.0401>, rot=<-0.0037 0.0213 0.298 -0.9543>}
 thumb_03_l {parent=thumb_02_l, loc=<2.2223 0.3682 -0.0907>, rot=<0.0118 -0.013 0.0066 -0.9998>}
 weapon_l {parent=hand_l, loc=<-0.1974 -3.4111 0.3811>, rot=<0 0 0 1>}
 lowerarm_twist_01_l {parent=lowerarm_l, loc=<18.316 0 0>, rot=<0 0 0 1>}
 lowerarm_twist_02_l {parent=lowerarm_l, loc=<10.6707 0 0>, rot=<0 0 0 1>}
 upperarm_twist_01_l {parent=upperarm_l, loc=<7.2135 -0.1134 0>, rot=<0 0 0 1>}
 upperarm_twist_02_l {parent=upperarm_l, loc=<19.7998 -0.1134 0>, rot=<0 0 0 1>}
 clavicle_r {parent=spine_05, loc=<5.4831 -0.1053 3.2946>, rot=<0.7289 0.0465 -0.6829 0.009>}
 upperarm_r {parent=clavicle_r, loc=<-9.8658 1.2136 0.4104>, rot=<0.0208 -0.486 -0.0242 -0.8734>}
 lowerarm_r {parent=upperarm_r, loc=<-26.4651 0.8345 -0.259>, rot=<0.0036 0.0161 -0.2298 -0.9731>}
 hand_r {parent=lowerarm_r, loc=<-22.7356 0.3203 0.6106>, rot=<-0.6784 -0.0223 0.0735 -0.7307>}
 index_metacarpal_r {parent=hand_r, loc=<-1.6534 -0.0181 2.9601>, rot=<0.0158 -0.0124 -0.0173 -0.9996>}
 index_01_r {parent=index_metacarpal_r, loc=<-8.3008 -0.0357 0.0524>, rot=<-0.0044 0.0485 0.0924 -0.9945>}
 index_02_r {parent=index_01_r, loc=<-3.3589 0.3003 -0.0453>, rot=<-0.002 0.0027 0.3225 -0.9466>}
 index_03_r {parent=index_02_r, loc=<-2.5578 0.123 0.0252>, rot=<-0.002 0.0067 0.2558 -0.9667>}
 middle_metacarpal_r {parent=hand_r, loc=<-2.1856 0.1726 0.5984>, rot=<0.006 0.0168 -0.0329 -0.9993>}
 middle_01_r {parent=middle_metacarpal_r, loc=<-7.3273 0.2447 -0.0129>, rot=<-0.0208 0.0042 0.0858 -0.9961>}
 middle_02_r {parent=middle_01_r, loc=<-4.0124 -0.14 -0.1175>, rot=<-0.0218 0.0258 0.3512 -0.9357>}
 middle_03_r {parent=middle_02_r, loc=<-2.6077 -0.0533 -0.1373>, rot=<0.0082 0.0188 0.1708 -0.9851>}
 pinky_metacarpal_r {parent=hand_r, loc=<-1.7617 0.9753 -3.3814>, rot=<-0.0397 0.0814 -0.069 -0.9935>}
 pinky_01_r {parent=pinky_metacarpal_r, loc=<-7.1707 0.4832 0.1094>, rot=<-0.0654 -0.0035 0.2242 -0.9723>}
 pinky_02_r {parent=pinky_01_r, loc=<-2.8274 -0.0892 0.1043>, rot=<0.0037 0.0011 0.2854 -0.9584>}
 pinky_03_r {parent=pinky_02_r, loc=<-2.2045 -0.031 0.0078>, rot=<-0.0002 0.0098 0.1445 -0.9895>}
 ring_metacarpal_r {parent=hand_r, loc=<-2.2073 0.3627 -1.3638>, rot=<-0.0284 0.0722 -0.0458 -0.9959>}
 ring_01_r {parent=ring_metacarpal_r, loc=<-7.0297 0.6197 0.1446>, rot=<-0.0489 -0.0252 0.1892 -0.9804>}
 ring_02_r {parent=ring_01_r, loc=<-3.6171 -0.4648 0.1027>, rot=<0 0 0.2806 -0.9598>}
 ring_03_r {parent=ring_02_r, loc=<-2.4153 -0.1141 -0.0199>, rot=<0 0 0.1189 -0.9929>}
 thumb_01_r {parent=hand_r, loc=<-1.3119 0.3192 2.459>, rot=<0.6546 -0.4042 0.0664 -0.6354>}
 thumb_02_r {parent=thumb_01_r, loc=<-5.9335 0.0447 -0.0401>, rot=<-0.0037 0.0213 0.298 -0.9543>}
 thumb_03_r {parent=thumb_02_r, loc=<-2.2223 -0.3682 0.0907>, rot=<0.0118 -0.013 0.0066 -0.9998>}
 weapon_r {parent=hand_r, loc=<0.1975 3.4112 -0.3811>, rot=<0 0 0 -1>}
 lowerarm_twist_01_r {parent=lowerarm_r, loc=<-18.3161 0 0>, rot=<0 0 0 1>}
 lowerarm_twist_02_r {parent=lowerarm_r, loc=<-10.6707 0 0>, rot=<0 0 0 1>}
 upperarm_twist_01_r {parent=upperarm_r, loc=<-7.2135 0.1134 0>, rot=<0 0 0 1>}
 upperarm_twist_02_r {parent=upperarm_r, loc=<-19.7998 0.1134 0>, rot=<0 0 0 1>}
 neck_01 {parent=spine_05, loc=<11.3138 0.1529 0>, rot=<0 0 -0.2003 -0.9797>}
 neck_02 {parent=neck_01, loc=<4.5627 0 0>, rot=<0 0 0.0504 -0.9987>}
 head {parent=neck_02, loc=<5.3914 0 0>, rot=<0 0 0.115 -0.9934>}
 dyn_head_offset {parent=head, loc=<7.124 1.7529 0.3511>, rot=<0.0029 -0.7071 0.7071 -0.0029>}
 dyn_hair_front_r_base {parent=dyn_head_offset, loc=<6.4022 -6.2082 -3.2521>, rot=<0.46 0.0402 0.0773 -0.8836>}
 dyn_hair_front_r_parent {parent=dyn_hair_front_r_base, loc=<0 0 0>, rot=<0 0 0 1>}
 dyn_hair_front_r_01 {parent=dyn_hair_front_r_parent, loc=<3.5008 -0.5456 0.2807>, rot=<-0.031 0.0583 0.0762 -0.9949>}
 dyn_hair_front_r_02 {parent=dyn_hair_front_r_01, loc=<4.1391 0.3227 0.5664>, rot=<0.0458 0.0839 -0.1612 -0.9823>}
 dyn_hair_front_r_03 {parent=dyn_hair_front_r_02, loc=<4.2917 1.7367 -0.1696>, rot=<-0.006 -0.1095 -0.2367 -0.9654>}
 dyn_hair_front_r_04 {parent=dyn_hair_front_r_03, loc=<4.517 0.5158 0.1541>, rot=<-0.1842 -0.3688 0.2435 -0.8779>}
 dyn_hair_front_r_05 {parent=dyn_hair_front_r_02, loc=<2.5493 1.6163 -2.2124>, rot=<-0.0091 0.0053 -0.1873 -0.9822>}
 dyn_hair_side_r_base {parent=dyn_head_offset, loc=<5.5545 -7.6463 -0.1471>, rot=<0.6526 -0.0197 0.1725 -0.7376>}
 dyn_hair_side_r_parent {parent=dyn_hair_side_r_base, loc=<0 0 0>, rot=<0 0 0 1>}
 dyn_hair_side_r_01 {parent=dyn_hair_side_r_parent, loc=<3.2966 -0.204 -0.3861>, rot=<0.1348 0.0533 0.0514 -0.9881>}
 dyn_hair_side_r_02 {parent=dyn_hair_side_r_01, loc=<4.378 0.0461 0.0431>, rot=<-0.1025 0.1211 -0.0144 -0.9872>}
 dyn_hair_side_r_03 {parent=dyn_hair_side_r_02, loc=<3.0545 0.5918 0.7978>, rot=<0.0609 0.2853 -0.0906 -0.9522>}
 dyn_hair_side_l_base {parent=dyn_head_offset, loc=<4.8132 8.1924 -1.7754>, rot=<-0.6224 0.0432 -0.1414 -0.7686>}
 dyn_hair_side_l_parent {parent=dyn_hair_side_l_base, loc=<0 0 0>, rot=<0 0 0 1>}
 dyn_hair_side_l_01 {parent=dyn_hair_side_l_parent, loc=<5.2161 0.1321 1.2602>, rot=<0.0218 -0.0297 0.0048 -0.9993>}
 dyn_hair_side_l_02 {parent=dyn_hair_side_l_01, loc=<5.6631 0.2205 0.6063>, rot=<-0.057 0.0753 0.3013 -0.9488>}
 dyn_hair_side_l_03 {parent=dyn_hair_side_l_02, loc=<4.1468 -0.8448 0.9191>, rot=<0.3965 0.166 0.4021 -0.8084>}
 dyn_hair_side_l_04 {parent=dyn_hair_side_l_03, loc=<3.0067 0.0009 0.0388>, rot=<0.3416 0.04 -0.2032 -0.9167>}
 dyn_hair_front_l_base {parent=dyn_head_offset, loc=<6.52 6.0119 -5.9608>, rot=<-0.438 0.0699 -0.1775 -0.8785>}
 dyn_hair_front_l_parent {parent=dyn_hair_front_l_base, loc=<0 0 0>, rot=<0 0 0 1>}
 dyn_hair_front_l_01 {parent=dyn_hair_front_l_parent, loc=<3.7634 -0.4615 0.2777>, rot=<-0.0717 0.2092 0.018 -0.9751>}
 dyn_hair_front_l_02 {parent=dyn_hair_front_l_01, loc=<3.637 -0.5485 0.6669>, rot=<0.2195 0.026 0.0416 -0.9744>}
 dyn_hair_front_l_03 {parent=dyn_hair_front_l_01, loc=<4.9118 0.7557 0.288>, rot=<0.3205 -0.0006 0.0603 -0.9453>}
 dyn_hair_back_r_base {parent=dyn_head_offset, loc=<1.5993 -5.2068 6.3144>, rot=<0.8949 -0.0457 0.1223 -0.4268>}
 dyn_hair_back_r_parent {parent=dyn_hair_back_r_base, loc=<0 0 0>, rot=<0 0 0 1>}
 dyn_hair_back_r_01 {parent=dyn_hair_back_r_parent, loc=<5.7833 0.1677 0.6632>, rot=<0.0059 -0.0229 0.0235 -0.9994>}
 dyn_hair_back_r_02 {parent=dyn_hair_back_r_01, loc=<5.8698 0.6309 -0.2256>, rot=<-0.0263 -0.1089 -0.0276 -0.9933>}
 dyn_hair_back_r_03 {parent=dyn_hair_back_r_02, loc=<3.4964 0.4553 0.0164>, rot=<-0.0068 0.1328 -0.0507 -0.9898>}
 dyn_hair_back_m_base {parent=dyn_head_offset, loc=<1.3294 0.4978 9.1177>, rot=<0.9962 0.0361 0.0793 -0.0087>}
 dyn_hair_back_m_parent {parent=dyn_hair_back_m_base, loc=<0 0 0>, rot=<0 0 0 1>}
 dyn_hair_back_m_01 {parent=dyn_hair_back_m_parent, loc=<5.7109 0.5205 -0.297>, rot=<0.0465 -0.0574 -0.0125 -0.9972>}
 dyn_hair_back_m_02 {parent=dyn_hair_back_m_01, loc=<5.5755 0.4101 0.0157>, rot=<-0.0697 -0.0344 0.0061 -0.997>}
 dyn_hair_back_m_03 {parent=dyn_hair_back_m_02, loc=<5.1957 1.3824 0.2175>, rot=<0.05 0.1228 -0.0129 -0.9911>}
 dyn_hair_back_m_04 {parent=dyn_hair_back_m_03, loc=<2.8237 -1.3666 1.3085>, rot=<-0.0925 0.0534 0.0996 -0.9893>}
 dyn_hair_back_m_05 {parent=dyn_hair_back_m_03, loc=<3.3789 2.0947 1.2289>, rot=<-0.0915 0.0381 0.082 -0.9917>}
 dyn_hair_back_l_base {parent=dyn_head_offset, loc=<0.9765 6.1631 6.071>, rot=<0.9182 0.0857 0.1097 0.371>}
 dyn_hair_back_l_parent {parent=dyn_hair_back_l_base, loc=<0 0 0>, rot=<0 0 0 1>}
 dyn_hair_back_l_01 {parent=dyn_hair_back_l_parent, loc=<6.2484 0.76 0.183>, rot=<-0.024 -0.0937 -0.0497 -0.9941>}
 dyn_hair_back_l_02 {parent=dyn_hair_back_l_01, loc=<5.9958 2.5683 0.5888>, rot=<-0.2382 0.0506 -0.1011 -0.9646>}
 dyn_hair_back_l_03 {parent=dyn_hair_back_l_02, loc=<4.234 0.3773 -0.4692>, rot=<-0.0474 0.0898 -0.0313 -0.9943>}
 dyn_hair_back_l_04 {parent=dyn_hair_back_l_01, loc=<6.1298 -1.7003 0.5894>, rot=<-0.0478 0.0349 -0.0213 -0.998>}
 dyn_hair_back_l_05 {parent=dyn_hair_back_l_04, loc=<4.4382 -0.4601 0.8138>, rot=<0.2574 -0.0844 -0.2351 -0.9335>}
 hairMaster {parent=head, loc=<6.8754 -5.3644 0.001>, rot=<0.0041 -1 0 0>}
 base_backHair_L_01 {parent=hairMaster, loc=<7.4778 0.604 7.5>, rot=<0 0 0.1768 -0.9842>}
 dyn_backHair_L_01 {parent=base_backHair_L_01, loc=<0 0 0>, rot=<0 0 0 1>}
 dyn_backHair_L_02 {parent=dyn_backHair_L_01, loc=<8.5316 -0.4247 0>, rot=<0 0 -0.0963 -0.9954>}
 base_backHair_R_01 {parent=hairMaster, loc=<7.451 0.5783 -7.2549>, rot=<0 0 0.1884 -0.9821>}
 dyn_backHair_R_01 {parent=base_backHair_R_01, loc=<0 0 0>, rot=<0 0 0 1>}
 dyn_backHair_R_02 {parent=dyn_backHair_R_01, loc=<8.4677 0.5899 0>, rot=<0 0 -0.1261 -0.992>}
 base_backHair_R_03 {parent=dyn_backHair_R_01, loc=<6.5486 4.1125 -3.946>, rot=<0 0 -0.1261 -0.992>}
 dyn_backHair_R_03 {parent=base_backHair_R_03, loc=<0 0 0>, rot=<0 0 0 1>}
 base_backHair_M_01 {parent=hairMaster, loc=<7.498 -1.6739 0>, rot=<0 0 0.1777 -0.9841>}
 dyn_backHair_M_01 {parent=base_backHair_M_01, loc=<0 0 0>, rot=<0 0 0 1>}
 dyn_backHair_M_02 {parent=dyn_backHair_M_01, loc=<8.8479 0.0132 0>, rot=<0 0 -0.0941 -0.9956>}
 thigh_l {parent=pelvis, loc=<-1.0855 -0.2109 -8.7053>, rot=<0.0009 0.0355 0.0155 -0.9993>}
 calf_l {parent=thigh_l, loc=<-42.5284 -0.0019 -0.7297>, rot=<0.0006 -0.0161 -0.0391 -0.9991>}
 calf_twist_01_l {parent=calf_l, loc=<-35.8483 -1.9197 -0.5794>, rot=<0.0237 0.0095 -0.0035 -0.9997>}
 calf_twist_02_l {parent=calf_l, loc=<-19.1546 -1.7943 -0.2638>, rot=<0.0237 0.0095 -0.0035 -0.9997>}
 foot_l {parent=calf_l, loc=<-46.7402 -2.5014 -0.592>, rot=<0.0427 -0.0238 0.0353 -0.9982>}
 ball_l {parent=foot_l, loc=<-3.5658 10.8066 -0.1544>, rot=<0.0001 -0.0001 -0.7063 -0.7079>}
 thigh_twist_01_l {parent=thigh_l, loc=<-27.952 0 0>, rot=<0.0013 0 0 -1>}
 thigh_r {parent=pelvis, loc=<-1.0853 -0.2109 8.7053>, rot=<0.0355 -0.0009 0.9993 0.0155>}
 calf_r {parent=thigh_r, loc=<42.5286 0.0019 0.7297>, rot=<0.0006 -0.0161 -0.0391 -0.9991>}
 calf_twist_01_r {parent=calf_r, loc=<35.8485 1.9197 0.5794>, rot=<0.0237 0.0095 -0.0035 -0.9997>}
 calf_twist_02_r {parent=calf_r, loc=<19.1546 1.7943 0.2638>, rot=<0.0237 0.0095 -0.0035 -0.9997>}
 foot_r {parent=calf_r, loc=<46.7403 2.5014 0.592>, rot=<0.0427 -0.0238 0.0353 -0.9982>}
 ball_r {parent=foot_r, loc=<3.5658 -10.8066 0.1544>, rot=<0.0001 -0.0001 -0.7063 -0.7079>}
 thigh_twist_01_r {parent=thigh_r, loc=<27.952 0 0>, rot=<0.0013 0 0 -1>}
 attach {parent=root, loc=<0 0 0>, rot=<0 0 0 1>}
 ik_foot_root {parent=root, loc=<0 0 0>, rot=<0 0 0 1>}
 ik_foot_l {parent=ik_foot_root, loc=<14.8262 -7.3545 8.296>, rot=<0.0357 0.7045 0.0284 -0.7082>}
 ik_foot_r {parent=ik_foot_root, loc=<-14.8262 -7.3546 8.2958>, rot=<0.7082 0.0284 -0.7045 0.0357>}
 ik_hand_root {parent=root, loc=<0 0 0>, rot=<0 0 0 1>}
 ik_hand_gun {parent=ik_hand_root, loc=<-35.2227 3.1613 94.1496>, rot=<0.5451 0.2436 0.5214 -0.6097>}
 ik_hand_l {parent=ik_hand_gun, loc=<23.7864 63.4948 19.1095>, rot=<-0.3377 -0.9013 -0.2713 0>}
 ik_hand_r {parent=ik_hand_gun, loc=<0 0 0>, rot=<0 0 0 1>}
]
variables [
 GetPistol {type=bool, default=False}
 isSkydive {type=bool, default=False}
 isTargeting {type=bool, default=False}
]
