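# Physics asset for the back hair of the female medium Cavalry hat-off character part, named as a hair physics test.
# Left, middle and right chains each constrain two dynamic capsule segments in series from a tiny base capsule.
# Left and middle chains lock twist and swing mainly on one axis, while the right chain uses different limits without that lock.
# Capsules on head, neck, upper spine, clavicles, upper arms and hairMaster have no constraints, forming body collision shapes.
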PhysicsAsset /BRCosmetics/Characters/Player/Female/Medium/Bodies/F_MED_Cavalry/Meshes/Parts/F_MED_Cavalry_Hat_Off_TestHairPhysics
bodies {
 base_backHair_L_01 [{shapeType=Capsule, length=0.505, radius=0.505, shapeName=base_backHair_L_01_capsule, rotation=<rot 90 0 0>}]
 base_backHair_M_01 [{shapeType=Capsule, length=0.505, radius=0.505, shapeName=base_backHair_M_01_capsule, rotation=<rot 90 0 0>}]
 base_backHair_R_01 [{shapeType=Capsule, length=0.505, radius=0.505, shapeName=base_backHair_R_01_capsule, rotation=<rot 90 0 0>}]
 clavicle_l [{shapeType=Capsule, length=5.5024, radius=5.4985, shapeName=clavicle_l_capsule, center=<6.5471 0 0>, rotation=<rot 90 0 0>}]
 clavicle_r [{shapeType=Capsule, length=14.9842, radius=5.4985, shapeName=clavicle_r_capsule, center=<-7.9238 0.5794 0.2587>, rotation=<rot 90 0 180>}]
 dyn_backHair_L_01 [{shapeType=Capsule, length=4.0002, radius=3.3107, shapeName=dyn_backHair_L_01_capsule, center=<4.9351 -1.7621 0>, rotation=<rot 90 0 0>}]
 dyn_backHair_L_02 [{
   shapeType=Capsule
   length=4.2519
   radius=2.6461
   shapeName=dyn_backHair_L_02_capsule
   center=<3.4826 -0.4768 0>
   rotation=<rot -4.342 -179.9998 180>
  }]
 dyn_backHair_M_01 [{shapeType=Capsule, length=4.014, radius=3.2222, shapeName=dyn_backHair_M_01_capsule, center=<3.723 -1.7621 0>, rotation=<rot 90 0 0>}]
 dyn_backHair_M_02 [{shapeType=Capsule, length=4.6439, radius=3.5948, shapeName=dyn_backHair_M_02_capsule, center=<3.6789 -1.4763 0>, rotation=<rot 90 0 0>}]
 dyn_backHair_R_01 [{shapeType=Capsule, length=2.95, radius=3.868, shapeName=dyn_backHair_R_01_capsule, center=<3.723 0.0109 -1.0032>, rotation=<rot 82.3471 0 0>}]
 dyn_backHair_R_02 [{
   shapeType=Capsule
   length=1.9005
   radius=3.0218
   shapeName=dyn_backHair_R_02_capsule
   center=<1.6998 -1.3936 -1.5665>
   rotation=<rot 90 180 -171.4302>
  }]
 hairMaster [{shapeType=Capsule, length=5.7482, radius=2.1572, shapeName=hairMaster_capsule, rotation=<rot 90 0 0>}]
 head [{shapeType=Capsule, length=2.767, radius=6.5666, shapeName=head_capsule, center=<1.7149 0.7119 0>, rotation=<rot 90 0 0>}]
 neck_01 [{shapeType=Capsule, length=0, radius=4.3697, shapeName=neck_01_capsule, rotation=<rot 90 0 0>}]
 neck_02 [{shapeType=Capsule, length=0.505, radius=4.5902, shapeName=neck_02_capsule, rotation=<rot 90 0 0>}]
 spine_05 [{shapeType=Capsule, length=10, radius=10, shapeName=spine_05_capsule, center=<0 1.839 0>}]
 upperarm_l [{shapeType=Capsule, length=19.4647, radius=5.4985, shapeName=upperarm_l_capsule, center=<13.8389 0 0>, rotation=<rot 90 0 0>}]
 upperarm_r [{shapeType=Capsule, length=19.4647, radius=5.4985, shapeName=upperarm_r_capsule, center=<-13.8389 0 0>, rotation=<rot 90 0 180>}]
}
constraints [
 {
  bone1Name=dyn_backHair_L_01
  bone2Name=base_backHair_L_01
  swing1LimitDegrees=30
  swing2LimitDegrees=0
  twistLimitDegrees=45
  twistMotion=Locked
 }
 {
  bone1Name=dyn_backHair_L_02
  bone2Name=dyn_backHair_L_01
  swing1LimitDegrees=30
  swing2LimitDegrees=0
  twistLimitDegrees=45
  twistMotion=Locked
 }
 {
  bone1Name=dyn_backHair_M_01
  bone2Name=base_backHair_M_01
  swing1LimitDegrees=30
  swing2LimitDegrees=0
  twistLimitDegrees=45
  twistMotion=Locked
 }
 {
  bone1Name=dyn_backHair_M_02
  bone2Name=dyn_backHair_M_01
  swing1LimitDegrees=30
  swing2LimitDegrees=0
  twistLimitDegrees=45
  twistMotion=Locked
 }
 {bone1Name=dyn_backHair_R_01, bone2Name=base_backHair_R_01, swing1LimitDegrees=35, swing2LimitDegrees=5, twistLimitDegrees=20}
 {bone1Name=dyn_backHair_R_02, bone2Name=dyn_backHair_R_01, swing1LimitDegrees=35, swing2LimitDegrees=5, twistLimitDegrees=0}
]
