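# Physics asset for the F_MED_Cavalry female player body, named as a test, setting up dynamic hair with upper-body collision shapes.
# Four hair chains (front and side, left and right) run as capsules from small box base bodies, with both front chains forking near the tips.
# Constraints give narrow twist and swing limits, and lock one swing axis at the front-right root.
# Capsules cover head, neck, upper spine, both clavicles and only the right upper arm.
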PhysicsAsset /BRCosmetics/Characters/Player/Female/Medium/Bodies/F_MED_Cavalry/Meshes/Parts/F_MED_Cavalry_Physics_test
bodies {
 clavicle_l [{shapeType=Capsule, length=10, radius=5.5094, center=<6.3939 -1.0067 -1.1451>, rotation=<rot -85.8476 -0.0001 0.0001>}]
 clavicle_r [{shapeType=Capsule, length=14.8176, radius=6.5, center=<-4.6354 2.3774 1.958>, rotation=<rot -87.1875 0.0012 -0.0012>}]
 dyn_hair_front_l_01 [{shapeType=Capsule, length=2.75, radius=1.375, center=<1.9269 0.0001 0>, rotation=<rot 90 0 0>}]
 dyn_hair_front_l_02 [{shapeType=Capsule, length=2.25, radius=1, center=<1.7326 0 0>, rotation=<rot 90 0 0>}]
 dyn_hair_front_l_03 [{shapeType=Capsule, length=3.375, radius=0.75, center=<1.7897 0.0235 -1.1718>, rotation=<rot 52.545 0.0001 0.0001>}]
 dyn_hair_front_l_base [{shapeType=Box, extentX=0.75, extentY=0.5, extentZ=0.5}]
 dyn_hair_front_l_parent [{shapeType=Capsule, length=2.625, radius=1.5, center=<2.3964 -0.0001 0>, rotation=<rot 90 0 0>}]
 dyn_hair_front_r_01 [{shapeType=Capsule, length=2.875, radius=1.875, center=<2.6105 0.0001 0>, rotation=<rot 90 0 0>}]
 dyn_hair_front_r_02 [{shapeType=Capsule, length=2.625, radius=2.125, center=<1.9632 0 0>, rotation=<rot 90 0 0>}]
 dyn_hair_front_r_03 [{shapeType=Capsule, length=3.25, radius=1.5, center=<2.127 0.54 0.5896>, rotation=<rot 90 -32.3803 -43.6305>}]
 dyn_hair_front_r_04 [{shapeType=Capsule, length=1.375, radius=1.125, center=<1.4484 0 -0.7346>, rotation=<rot 73.1249 0.0002 0.0002>}]
 dyn_hair_front_r_05 [{shapeType=Capsule, length=2, radius=0.75, center=<1.5155 0 0>, rotation=<rot 90 0 0>}]
 dyn_hair_front_r_base [{shapeType=Box, extentX=0.75, extentY=0.5, extentZ=0.5}]
 dyn_hair_front_r_parent [{shapeType=Capsule, length=3.25, radius=1.5, center=<2.21 0 -0.0001>, rotation=<rot 90 0 0>}]
 dyn_hair_side_l_01 [{shapeType=Capsule, length=2.125, radius=2.875, center=<3.4512 0 0.0001>, rotation=<rot 90 0 0>}]
 dyn_hair_side_l_02 [{shapeType=Capsule, length=2.0359, radius=2.7141, center=<1.2845 0 -0.0001>, rotation=<rot 90 -9.4623 13.929>}]
 dyn_hair_side_l_03 [{shapeType=Capsule, length=1.236, radius=1.514, center=<0.955 0 -0.3317>, rotation=<rot 90 0 0>}]
 dyn_hair_side_l_04 [{shapeType=Capsule, length=1.8814, radius=3, center=<2.1954 0.5831 -0.7995>, rotation=<rot 90 0 -13.4886>}]
 dyn_hair_side_l_base [{shapeType=Box, extentX=0.75, extentY=0.5, extentZ=0.5}]
 dyn_hair_side_l_parent [{shapeType=Capsule, length=2.25, radius=2, center=<2.6302 0 0.63>, rotation=<rot 90 0 0>}]
 dyn_hair_side_r_01 [{shapeType=Capsule, length=3.875, radius=1.5, center=<1.8633 0 0.0001>, rotation=<rot 90 0 0>}]
 dyn_hair_side_r_02 [{shapeType=Capsule, length=2.625, radius=1.75, center=<1.9399 0.4107 0.8238>, rotation=<rot 90 -159.444 -177.2946>}]
 dyn_hair_side_r_03 [{shapeType=Capsule, length=3.625, radius=1.125, center=<2.389 1.1723 0.1087>, rotation=<rot 75.0839 -179.9999 152.458>}]
 dyn_hair_side_r_base [{shapeType=Box, extentX=0.75, extentY=0.5, extentZ=0.5}]
 dyn_hair_side_r_parent [{shapeType=Capsule, length=2.125, radius=1.375, center=<1.7271 0 0>, rotation=<rot 90 0 0>}]
 dyn_head_offset [{shapeType=Box, extentX=0.5, extentY=0.5, extentZ=0.5}]
 head [{shapeType=Capsule, length=5.5, radius=7.5, center=<-0.0002 2.1581 -0.0324>, rotation=<rot 89.3496 179.9998 179.9998>}]
 neck_02 [{shapeType=Capsule, length=7.75, radius=6, center=<-1.6187 0.1345 -0.0184>, rotation=<rot 89.3496 179.9998 179.9998>}]
 spine_04 [{shapeType=Capsule, length=20, radius=10, shapeName=spine_04_capsule, center=<2.4445 0 0>}]
 spine_05 [{shapeType=Capsule, length=18.75, radius=8, shapeName=CenterBody, center=<0 -0.4381 0>}]
 upperarm_r [{shapeType=Capsule, length=17.5, radius=6.5, shapeName=upperarm_r_capsule, center=<-11.6816 -0.0001 0>, rotation=<rot -0.0002 -89.9997 90>}]
}
constraints [
 {
  bone1Name=dyn_hair_front_r_parent
  bone2Name=dyn_hair_front_r_base
  swing1LimitDegrees=5
  swing1Motion=Locked
  swing2LimitDegrees=20
  twistLimitDegrees=3
 }
 {
  bone1Name=dyn_hair_front_r_01
  bone2Name=dyn_hair_front_r_parent
  swing1LimitDegrees=5
  swing1Motion=Locked
  swing2LimitDegrees=20
  twistLimitDegrees=3
 }
 {bone1Name=dyn_hair_front_r_02, bone2Name=dyn_hair_front_r_01, swing1LimitDegrees=5, swing2LimitDegrees=20, twistLimitDegrees=3}
 {bone1Name=dyn_hair_front_r_03, bone2Name=dyn_hair_front_r_02, swing1LimitDegrees=20, swing2LimitDegrees=20, twistLimitDegrees=3}
 {bone1Name=dyn_hair_front_r_04, bone2Name=dyn_hair_front_r_03, swing1LimitDegrees=20, swing2LimitDegrees=20, twistLimitDegrees=3}
 {bone1Name=dyn_hair_front_r_05, bone2Name=dyn_hair_front_r_02, swing1LimitDegrees=20, swing2LimitDegrees=20, twistLimitDegrees=3}
 {bone1Name=dyn_hair_side_r_parent, bone2Name=dyn_hair_side_r_base, swing1LimitDegrees=5, swing2LimitDegrees=20, twistLimitDegrees=3}
 {bone1Name=dyn_hair_side_r_01, bone2Name=dyn_hair_side_r_parent, swing1LimitDegrees=5, swing2LimitDegrees=20, twistLimitDegrees=3}
 {bone1Name=dyn_hair_side_r_02, bone2Name=dyn_hair_side_r_01, swing1LimitDegrees=5, swing2LimitDegrees=20, twistLimitDegrees=3}
 {bone1Name=dyn_hair_side_r_03, bone2Name=dyn_hair_side_r_02, swing1LimitDegrees=5, swing2LimitDegrees=20, twistLimitDegrees=3}
 {bone1Name=dyn_hair_side_l_parent, bone2Name=dyn_hair_side_l_base, swing1LimitDegrees=5, swing2LimitDegrees=20, twistLimitDegrees=3}
 {bone1Name=dyn_hair_side_l_01, bone2Name=dyn_hair_side_l_parent, swing1LimitDegrees=5, swing2LimitDegrees=20, twistLimitDegrees=3}
 {bone1Name=dyn_hair_side_l_02, bone2Name=dyn_hair_side_l_01, swing1LimitDegrees=5, swing2LimitDegrees=20, twistLimitDegrees=3}
 {bone1Name=dyn_hair_side_l_03, bone2Name=dyn_hair_side_l_02, swing1LimitDegrees=30, swing2LimitDegrees=30, twistLimitDegrees=3}
 {bone1Name=dyn_hair_side_l_04, bone2Name=dyn_hair_side_l_03, swing1LimitDegrees=20, swing2LimitDegrees=30, twistLimitDegrees=3}
 {bone1Name=dyn_hair_front_l_parent, bone2Name=dyn_hair_front_l_base, swing1LimitDegrees=5, swing2LimitDegrees=20, twistLimitDegrees=3}
 {bone1Name=dyn_hair_front_l_01, bone2Name=dyn_hair_front_l_parent, swing1LimitDegrees=5, swing2LimitDegrees=20, twistLimitDegrees=3}
 {bone1Name=dyn_hair_front_l_02, bone2Name=dyn_hair_front_l_01, swing1LimitDegrees=5, swing2LimitDegrees=10, twistLimitDegrees=3}
 {bone1Name=dyn_hair_front_l_03, bone2Name=dyn_hair_front_l_01, swing1LimitDegrees=5, swing2LimitDegrees=10, twistLimitDegrees=3}
]
