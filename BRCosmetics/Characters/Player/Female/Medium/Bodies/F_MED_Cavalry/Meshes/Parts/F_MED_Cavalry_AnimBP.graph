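# Character-part AnimBlueprint for the F_MED_Cavalry outfit, driving hair physics and corrective posing.
# It copies the parent pose, layers clavicles from the body mesh, then runs a Control Rig fed skydive, pistol and targeting states.
# Two RigidBody passes follow, the first using sine-and-Perlin fake wind as gravity while skydiving, then constraints adjust back-hair bones.
# Backbling flags and gravity/RBAN overrides are computed but unused in the AnimGraph.
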AnimBlueprint /BRCosmetics/Characters/Player/Female/Medium/Bodies/F_MED_Cavalry/Meshes/Parts/F_MED_Cavalry_AnimBP
target_skeleton=&/F_MED_Cavalry_Skeleton
graphs [
 AnimGraph {nodes [
   n1 = Get Body_Copy() -> (Body_Copy:Skeletal Mesh Component Budgeted Object Reference)
   n2 = Control Rig!(
    Source=Layered blend per bone(BasePose=Copy Pose From Mesh() {bUseAttachedParent=true}, BlendPoses_0=Copy Pose From Mesh(SourceMeshComponent=@n1)) {layerSetup [{branchFilters [{boneName=clavicle_l}, {boneName=clavicle_r}]}]}
    Alpha
    isSkydive:bool=Get bIsSkyDiving()
    GetPistol:bool=Get GetPistol() -> (GetPistol:bool)
    isTargeting:bool=NOT Boolean(A=Get bIsTargeting())
   ) -> (Pose) {
    alphaBoolBlend {blendOption=Linear}
    controlRigAssetReference {blueprintRigClass=&/F_MED_Cavalry_Hat_Off_TEstNewSetup_CtrlRig.F_MED_Cavalry_Hat_Off_TEstNewSetup_CtrlRig_C}
   }
   n3 = RigidBody!(ComponentPose=Local To Component(LocalPose=@n2), Alpha, OverrideWorldGravity:Vector=Get fakeWind() -> (fakeWind:Vector), ExternalForce) -> (Pose) {
    bOverrideWorldGravity=true
    overridePhysicsAsset=&/F_MED_Cavalry_Hat_Off_TestHairPhysics
    simulationSpace=BaseBoneSpace
    alphaBoolBlend {blendOption=Linear}
    baseBoneRef {boneName=spine_05}
    overrideWorldGravity {z=-980}
    simSpaceSettings {
     dampingAlpha=0.2
     maxAngularAcceleration=3
     maxAngularVelocity=1
     maxLinearAcceleration=1000
     maxLinearVelocity=500
     worldAlpha=1
     externalLinearDragV=<0.2 0.2 0.2>
    }
   }
   n4 = RigidBody!(ComponentPose=@n3, Alpha, ExternalForce) -> (Pose) {
    bOverrideWorldGravity=true
    overridePhysicsAsset=&/F_MED_Cavalry_Physics_test
    simulationSpace=BaseBoneSpace
    alphaBoolBlend {blendOption=Linear}
    baseBoneRef {boneName=spine_05}
    overrideWorldGravity {z=-980}
    simSpaceSettings {
     dampingAlpha=0.2
     maxAngularAcceleration=3
     maxAngularVelocity=1
     maxLinearAcceleration=500
     maxLinearVelocity=200
     worldAlpha=0.5
     externalLinearDragV=<0.2 0.2 0.2>
    }
   }
   n5 = "Constraint - base_backHair_L_01 (PP)"!(ComponentPose=@n4, Alpha=1, ConstraintWeights_0=0.5, ConstraintWeights_1:float=0.5) -> (Pose) {
    alphaBoolBlend {blendOption=Linear}
    boneToModify {boneName=base_backHair_L_01}
    constraintSetup [{targetBone {boneName=spine_05}}, {targetBone {boneName=head}}]
    constraintWeights [1,1]
   }
   n6 = "Constraint - base_backHair_M_01 (PP)"!(ComponentPose=@n5, Alpha=1, ConstraintWeights_0=0.5, ConstraintWeights_1:float=0.5) -> (Pose) {
    alphaBoolBlend {blendOption=Linear}
    boneToModify {boneName=base_backHair_M_01}
    constraintSetup [{targetBone {boneName=spine_05}}, {targetBone {boneName=head}}]
    constraintWeights [1,1]
   }
   n7 = "Constraint - base_backHair_R_01 (PP)"!(ComponentPose=@n6, Alpha=1, ConstraintWeights_0=0.5, ConstraintWeights_1:float=0.5) -> (Pose) {
    alphaBoolBlend {blendOption=Linear}
    boneToModify {boneName=base_backHair_R_01}
    constraintSetup [{targetBone {boneName=spine_05}}, {targetBone {boneName=head}}]
    constraintWeights [1,1]
   }
   n8 = "Constraint - dyn_backHair_L_02 (P)"(ComponentPose=@n7, Alpha=1, ConstraintWeights_0=1) {
    alphaBoolBlend {blendOption=Linear}
    boneToModify {boneName=dyn_backHair_L_02}
    constraintSetup [{targetBone {boneName=spine_05}}]
    constraintWeights [1]
   }
   n9 = "Constraint - dyn_backHair_M_02 (P)"(ComponentPose=@n8, Alpha=1, ConstraintWeights_0=1) {
    alphaBoolBlend {blendOption=Linear}
    boneToModify {boneName=dyn_backHair_M_02}
    constraintSetup [{targetBone {boneName=spine_05}}]
    constraintWeights [1]
   }
   n10 = "Constraint - dyn_backHair_R_02 (P)"(ComponentPose=@n9, Alpha=1, ConstraintWeights_0=1) {
    alphaBoolBlend {blendOption=Linear}
    boneToModify {boneName=dyn_backHair_R_02}
    constraintSetup [{targetBone {boneName=spine_05}}]
    constraintWeights [1]
   }
   Output Pose(Result=Component To Local(ComponentPose=@n10)) {layerGroup=None}
  ]}
 EventGraph {nodes [
   Event BlueprintInitializeAnimation()
   > n1 = Cast To PlayerPawn_Athena(Object=TryGetPawnOwner())
   > n2 = Set Pawn Owner(Pawn Owner:Player Pawn Athena Object Reference=@n1) -> (Output_Get:Player Pawn Athena Object Reference)
   > Set Body_Mesh(Body_Mesh=GetSkeletalMeshForPartType(self=@n2))
   n3 = Event BlueprintUpdateAnimation()
   > n4 = Sequence!(execute) -> (then_0, then_1, then_2:Exec)
   > n5 = Gravity Override State Function DEPRECATED(
    CharacterPartAnimInstance=Self-Reference()
    GameTimeInSeconds=GetGameTimeInSeconds()
    GravityOverrideParams=Get Gravity Override Params()
    RBAN Params=Get RBAN Params()
   )
   > Set Gravity Override(Gravity Override=@n5)
   > Set RBAN Settings Out(RBAN Settings Out:RBANSetting Struct Structure=@n5.RBAN SettingsOut) -> (Output_Get:RBANSetting Struct Structure)
   n6 = Sequence!(execute=@n4.then_1) -> (then_0, then_1)
   > n7 = Branch(Condition=Get bIsSkydiveFloating())
   > Set Set Gravity(Set Gravity:Vector="0,0,-200") -> (Output_Get:Vector)
   Set Set Gravity(execute=@n7.else, Set Gravity="0,0,-500")
   > Branch(Condition=Get bIsSkydiveDiveMode())
   > Set Set Gravity(Set Gravity="0,0,500")
   n8 = Branch(execute=@n6.then_1, Condition=Get bIsSkyDiving())
   n9 = BreakVector(InVec="20,100,800")
   n10 = GetGameTimeInSeconds()
   n11 = BreakRotator(InRot="4,8,1")
   n12 = "Sin (Radians)"(A=float + float(A=float * float(A=@n11.Yaw, B=@n10) {is=title K2Node_PromotableOperator}, B=@n3.DeltaTimeX))
   n13 = float * float(A=PerlinNoise1D!(Value:float=@n10) -> (ReturnValue:float), B=1) {is=title K2Node_CommutativeAssociativeBinaryOperator}
   n14 = float * float(A=@n12, B=@n13)
   n15 = float + float(A=float * float(A=float + float(A=@n12, B=@n14) {is=title K2Node_CommutativeAssociativeBinaryOperator}, B=@n9.Z))
   n16 = "Sin (Radians)"(A=float + float(A=float * float(A=@n11, B=@n10) {is=title K2Node_PromotableOperator}, B=@n3.DeltaTimeX) {is=title K2Node_PromotableOperator})
   n17 = float * float(A=@n16, B=@n13) {is=title K2Node_CommutativeAssociativeBinaryOperator}
   n18 = float + float(A=float * float(A=float + float(A=@n16, B=@n17) {is=title K2Node_CommutativeAssociativeBinaryOperator}, B=@n9))
   n19 = "Sin (Radians)"(A=float + float(A=float * float(A=@n11.Pitch, B=@n10) {is=title K2Node_PromotableOperator}, B=@n3.DeltaTimeX) {is=title K2Node_PromotableOperator})
   n20 = float * float(A=@n19, B=@n13) {is=title K2Node_CommutativeAssociativeBinaryOperator}
   n21 = float + float(A=float * float(A=float + float(A=@n19, B=@n20) {is=title K2Node_CommutativeAssociativeBinaryOperator}, B=@n9.Y))
   > Set fakeWind(fakeWind:Vector=vector + vector(A=MakeVector(X=@n18, Y=@n21, Z=@n15), B=Get Set Gravity() -> (Set Gravity:Vector))) -> (Output_Get:Vector)
   Set fakeWind(execute=@n8.else, fakeWind="0,0,-980")
   n22 = Cast To FortPlayerAnimInstance(execute=@n4.then_2, Object=GetAnimInstance(self=GetOwnerSkeletalMeshComponent()))
   > Set GetPistol(GetPistol:bool=false) -> (Output_Get:bool)
   > n23 = Switch on EFortWeaponCoreAnimation!(execute, Selection:EFortWeaponCoreAnimation Enum=Get WeaponCoreAnim(self=@n22)) -> (
    Melee:Exec
    Pistol:Exec
    Shotgun:Exec
    PaperBlueprint:Exec
    Rifle:Exec
    MeleeOneHand:Exec
    MachinePistol:Exec
    RocketLauncher:Exec
    GrenadeLauncher:Exec
    GoingCommando:Exec
    AssaultRifle:Exec
    TacticalShotgun:Exec
    SniperRifle:Exec
    TrapPlacement:Exec
    ShoulderLauncher:Exec
    AbilityDecoTool:Exec
    Crossbow:Exec
    C4:Exec
    RemoteControl:Exec
    DualWield:Exec
    AR_BullPup:Exec
    AR_ForwardGrip:Exec
    MedPackPaddles:Exec
    SMG_P90:Exec
    AR_DrumGun:Exec
    Consumable_Small:Exec
    Consumable_Large:Exec
    Balloon:Exec
    MountedTurret:Exec
    CreativeTool:Exec
    ExplosiveBow:Exec
    AshtonIndigo:Exec
    AshtonChicago:Exec
    MeleeDualWield:Exec
    Unarmed:Exec
   )
   Branch(execute=[@n23.DualWield, @n23.Pistol])
   > Set GetPistol(GetPistol=1)
   n24 = Event OnCharacterCustomizationLoaded()
   > n25 = Sequence()
   > Set Body_Copy(Body_Copy:Skeletal Mesh Component Budgeted Object Reference=Get Array Item with EFort Custom Part Type(Array=@n24.CustomMeshParts)) -> (Output_Get:Skeletal Mesh Component Budgeted Object Reference)
   n26 = CheckForBackbling()
   Set is Wearing Backbling(
    execute=@n25.then_1
    is Wearing Backbling:bool=AND Boolean!(A=@n26, B=NOT Boolean(A=@n26.bIsBackblingACape), C:bool=NOT Boolean(A=@n26.bIsBackblingLong)) -> (ReturnValue)
   ) -> (Output_Get:bool)
   > Set is Wearing Cape(is Wearing Cape:bool=OR Boolean(A=@n26.bIsBackblingACape, B=@n26.bIsBackblingLong)) -> (Output_Get:bool)
   n27 = OR Boolean(A=Get is Wearing Backbling() -> (is Wearing Backbling:bool), B=Get is Wearing Cape() -> (is Wearing Cape:bool))
   > Set is Wearing Backbling or Cape(is Wearing Backbling or Cape:bool=@n27) -> (Output_Get:bool)
   > Set is Wearing Backbling or Cape Alpha(is Wearing Backbling or Cape Alpha:double="To Float (Boolean)"(InBool=@n27)) -> (Output_Get:double)
   Reroute Node(InputPin:Wildcard) -> (OutputPin:Wildcard)
  ]}
]
variables [
 ADS_Alpha {type=number}
 Body_Copy {type=SkeletalMeshComponentBudgeted}
 Body_Mesh {type=SkeletalMeshComponent}
 CrouchingMoving2D {type=boolean}
 CrouchTargeting {type=boolean}
 fakeWind {type=Vector}
 GetPistol {type=boolean}
 Gravity Override {type=Vector}
 Gravity Override Params {
  type=GravityOverrideParamsStruct
  default {
   skydive_DivingDown {jointName=root, gravityOverride=<0 -700 100>, windAmplitude=<200 -500 400>, windFrequency=<1.5 7 8>, windOffset=<0 -600 400>}
   skydive_FloatingDown {windFrequency=<0.5 0.5 0.5>}
   skydive_GoingUp {gravityOverride=<0 -300 100>}
  }
 }
 HeadRotate_Alpha {type=number}
 is Wearing Backbling {type=boolean}
 is Wearing Backbling or Cape {type=boolean}
 is Wearing Backbling or Cape Alpha {type=number}
 is Wearing Cape {type=boolean}
 isBodyAttached {type=boolean}
 NotInSkydive {type=boolean}
 Pawn Owner {type=PlayerPawn_Athena_C}
 RBAN Params {
  type=RBANParamsStruct
  default {
   dBNO {simSpaceSettings {
     maxAngularAcceleration=1000
     maxAngularVelocity=3
     maxLinearAcceleration=300
     maxLinearVelocity=700
     velocityScaleZ=0
     externalLinearDragV=<0 0.5 0>
    }}
   dBNO_Carried {simSpaceSettings {
     maxAngularAcceleration=1000
     maxAngularVelocity=3
     maxLinearAcceleration=300
     maxLinearVelocity=700
     velocityScaleZ=0
     externalLinearDragV=<0 0.5 0>
    }}
   emoteOrMelee {simSpaceSettings {
     maxAngularAcceleration=1000
     maxAngularVelocity=3
     maxLinearAcceleration=300
     maxLinearVelocity=700
     velocityScaleZ=0
     externalLinearDragV=<0 0.5 0>
    }}
   front_End {simSpaceSettings {
     maxAngularAcceleration=1000
     maxAngularVelocity=3
     maxLinearAcceleration=300
     maxLinearVelocity=700
     velocityScaleZ=0
     externalLinearDragV=<0 0.5 0>
    }}
   onGround_Crouch_Moving {simSpaceSettings {
     maxAngularAcceleration=1000
     maxAngularVelocity=3
     maxLinearAcceleration=300
     maxLinearVelocity=700
     velocityScaleZ=0
     externalLinearDragV=<0 0.5 0>
    }}
   onGround_Falling {simSpaceSettings {
     maxAngularAcceleration=1000
     maxAngularVelocity=3
     maxLinearAcceleration=300
     maxLinearVelocity=700
     velocityScaleZ=0
     externalLinearDragV=<0 0.5 0>
    }}
   onGround_Moving {simSpaceSettings {
     maxAngularAcceleration=1000
     maxAngularVelocity=3
     maxLinearAcceleration=300
     maxLinearVelocity=700
     velocityScaleZ=0
     externalLinearDragV=<0 0.5 0>
    }}
   onGround_Standing {simSpaceSettings {
     maxAngularAcceleration=1000
     maxAngularVelocity=3
     maxLinearAcceleration=300
     maxLinearVelocity=700
     velocityScaleZ=0
     externalLinearDragV=<0 0.5 0>
    }}
   rail_Grinding {simSpaceSettings {
     maxAngularAcceleration=1000
     maxAngularVelocity=3
     maxLinearAcceleration=300
     maxLinearVelocity=700
     velocityScaleZ=0
     worldAlpha=0.1
     externalLinearDragV=<0 0.5 0>
    }}
   riding_Moving {simSpaceSettings {
     maxAngularAcceleration=1000
     maxAngularVelocity=3
     maxLinearAcceleration=300
     maxLinearVelocity=700
     velocityScaleZ=0
     externalLinearDragV=<0 0.5 0>
    }}
   riding_Moving_Targeting {simSpaceSettings {
     maxAngularAcceleration=1000
     maxAngularVelocity=3
     maxLinearAcceleration=300
     maxLinearVelocity=700
     velocityScaleZ=0
     externalLinearDragV=<0 0.5 0>
    }}
   riding_Standing {simSpaceSettings {
     maxAngularAcceleration=1000
     maxAngularVelocity=3
     maxLinearAcceleration=300
     maxLinearVelocity=700
     velocityScaleZ=0
     externalLinearDragV=<0 0.5 0>
    }}
   skydive_DivingDown {simSpaceSettings {
     maxAngularAcceleration=1000
     maxAngularVelocity=3
     maxLinearAcceleration=300
     maxLinearVelocity=700
     velocityScaleZ=0
     externalLinearDragV=<0 0.5 0>
    }}
   skydive_FloatingDown {
    componentLinearAccScale=<0 0.5 0>
    componentLinearVelScale=<0 0.5 0>
    simSpaceSettings {
     maxAngularAcceleration=1000
     maxAngularVelocity=3
     maxLinearAcceleration=300
     maxLinearVelocity=700
     velocityScaleZ=0
     externalLinearDragV=<0 0.5 0>
    }
   }
   skydive_GoingUp {simSpaceSettings {
     maxAngularAcceleration=1000
     maxAngularVelocity=3
     maxLinearAcceleration=300
     maxLinearVelocity=700
     velocityScaleZ=0
     externalLinearDragV=<0 0.5 0>
    }}
   skydive_Parachuting {
    componentLinearAccScale=<0 0.1 0.3>
    componentLinearVelScale=<0 0.1 0.3>
    simSpaceSettings {
     maxAngularAcceleration=1000
     maxAngularVelocity=3
     maxLinearAcceleration=300
     maxLinearVelocity=700
     velocityScaleZ=0
     externalLinearDragV=<0 0.5 0>
    }
   }
   skydive_Surfing {simSpaceSettings {
     maxAngularAcceleration=1000
     maxAngularVelocity=3
     maxLinearAcceleration=300
     maxLinearVelocity=700
     velocityScaleZ=0
     externalLinearDragV=<0 0.5 0>
    }}
   swimming {simSpaceSettings {
     maxAngularAcceleration=1000
     maxAngularVelocity=3
     maxLinearAcceleration=300
     maxLinearVelocity=700
     velocityScaleZ=0
     externalLinearDragV=<0 0.5 0>
    }}
   tactical_Sprint {simSpaceSettings {
     maxAngularAcceleration=1000
     maxAngularVelocity=3
     maxLinearAcceleration=300
     maxLinearVelocity=700
     velocityScaleZ=0
     externalLinearDragV=<0 0.5 0>
    }}
   ziplining {simSpaceSettings {
     maxAngularAcceleration=1000
     maxAngularVelocity=3
     maxLinearAcceleration=300
     maxLinearVelocity=700
     velocityScaleZ=0
     externalLinearDragV=<0 0.5 0>
    }}
  }
 }
 RBAN Settings Out {
  type=RBANSettingStruct
  default {
   componentLinearAccScale=<0 0 0>
   componentLinearVelScale=<0 0 0>
   simSpaceSettings {
    dampingAlpha=1
    maxAngularAcceleration=1000
    maxAngularVelocity=3
    maxLinearAcceleration=300
    maxLinearVelocity=700
    velocityScaleZ=0
    worldAlpha=1
    externalAngularVelocity=<0 0 0>
    externalLinearDragV=<0 0.5 0>
    externalLinearVelocity=<0 0 0>
   }
  }
 }
 Set Gravity {type=Vector}
 upperarm_l_Rotation {type=Rotator}
 upperarm_l_Translation {type=Vector}
 upperarm_r_Rotation {type=Rotator}
 upperarm_r_Translation {type=Vector}
]
