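# Physics asset for the Frontend version of the female medium Cavalry character, setting up hair dynamics around the upper body.
# Seven front, side and back hair chains use capsule bodies with tight swing and twist limits, some branching, with twist locked on back-right.
# Box bodies at each chain base and head offset, plus capsules for head, neck, clavicles and upper spine, connect to spine_04 through angularly free constraints.
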PhysicsAsset /BRCosmetics/Characters/Player/Female/Medium/Bodies/F_MED_Cavalry/Meshes/Parts/F_MED_Cavalry_Frontend_Physics
bodies {
 clavicle_l [{shapeType=Capsule, length=3.4794, radius=7.2519, center=<6.4482 -0.3686 -2.4355>, rotation=<rot 85.606 0 0>}]
 clavicle_r [{shapeType=Capsule, length=3.4794, radius=7.2519, center=<-6.4482 0.3686 2.4355>, rotation=<rot 85.606 0 0>}]
 dyn_hair_back_l_01 [{shapeType=Capsule, length=2.375, radius=1.5, center=<2.2081 0 0>, rotation=<rot 90 0 0>}]
 dyn_hair_back_l_02 [{shapeType=Capsule, length=2.125, radius=1.375, center=<1.98 0 0>, rotation=<rot 90 0 0>}]
 dyn_hair_back_l_03 [{shapeType=Capsule, length=4.125, radius=2.375, center=<2.7807 1.3315 0.0211>, rotation=<rot 80.2642 -65.8502 -91.3003>}]
 dyn_hair_back_l_04 [{shapeType=Capsule, length=1.375, radius=2, center=<2.3249 0 0.0001>, rotation=<rot 90 0 0>}]
 dyn_hair_back_l_05 [{shapeType=Capsule, length=6.125, radius=1.375, center=<2.7524 0 0.3266>, rotation=<rot 68.1208 -179.9998 -180>}]
 dyn_hair_back_l_base [{shapeType=Box, extentX=0.75, extentY=0.5, extentZ=0.5}]
 dyn_hair_back_l_parent [{shapeType=Capsule, length=2.375, radius=1.5, center=<2.1211 -0.0001 0>, rotation=<rot 90 0 0>}]
 dyn_hair_back_m_01 [{shapeType=Capsule, length=3.75, radius=1.5, center=<2.91 0 0>, rotation=<rot 90 0 0>}]
 dyn_hair_back_m_02 [{shapeType=Capsule, length=3.75, radius=1.5, center=<2.7984 0 0.0001>, rotation=<rot 90 0 0>}]
 dyn_hair_back_m_03 [{shapeType=Capsule, length=1.875, radius=1.125, center=<2.6073 0 0>, rotation=<rot 90 0 0>}]
 dyn_hair_back_m_04 [{shapeType=Capsule, length=2.5, radius=1, center=<1.456 -0.3663 -0.0001>, rotation=<rot 90 180 -160.6366>}]
 dyn_hair_back_m_05 [{shapeType=Capsule, length=5.5047, radius=2.7812, center=<1.1528 -0.9152 -0.7277>, rotation=<rot 90 -172.875 -162.7002>}]
 dyn_hair_back_m_base [{shapeType=Box, extentX=0.75, extentY=0.5, extentZ=0.5}]
 dyn_hair_back_m_parent [{shapeType=Capsule, length=3.75, radius=1.5, center=<2.9606 0 -0.0001>, rotation=<rot 90 0 0>}]
 dyn_hair_back_r_01 [{shapeType=Capsule, length=3.6164, radius=2.4312, center=<3.9772 0.0001 0>, rotation=<rot 90 0 0>}]
 dyn_hair_back_r_02 [{shapeType=Capsule, length=4.4324, radius=2.3449, center=<2.5778 -1.1776 0>, rotation=<rot 90 14.0363 82.1449>}]
 dyn_hair_back_r_03 [{shapeType=Capsule, length=6.7391, radius=1.9821, center=<1.448 -0.9612 0.3431>, rotation=<rot 86.5252 109.382 166.0901>}]
 dyn_hair_back_r_base [{shapeType=Box, extentX=0.75, extentY=0.5, extentZ=0.5}]
 dyn_hair_back_r_parent [{shapeType=Capsule, length=3.5, radius=1.125, center=<2.9821 0 0>, rotation=<rot 90 0 0>}]
 dyn_hair_front_l_01 [{shapeType=Capsule, length=2.75, radius=1.375, center=<1.9269 0.0001 0>, rotation=<rot 90 0 0>}]
 dyn_hair_front_l_02 [{shapeType=Capsule, length=2.25, radius=1, center=<1.7326 0 0>, rotation=<rot 90 0 0>}]
 dyn_hair_front_l_03 [{shapeType=Capsule, length=3.375, radius=0.75, center=<1.7897 0.0235 -1.1718>, rotation=<rot 52.545 0.0001 0.0001>}]
 dyn_hair_front_l_base [{shapeType=Box, extentX=0.75, extentY=0.5, extentZ=0.5}]
 dyn_hair_front_l_parent [{shapeType=Capsule, length=2.625, radius=1.5, center=<2.3964 -0.0001 0>, rotation=<rot 90 0 0>}]
 dyn_hair_front_r_01 [{shapeType=Capsule, length=2.875, radius=1.875, center=<2.6105 0.0001 0>, rotation=<rot 90 0 0>}]
 dyn_hair_front_r_02 [{shapeType=Capsule, length=2.625, radius=2.125, center=<1.9632 0 0>, rotation=<rot 90 0 0>}]
 dyn_hair_front_r_03 [{shapeType=Capsule, length=3.25, radius=1.5, center=<2.127 0 0>, rotation=<rot 90 0 0>}]
 dyn_hair_front_r_04 [{shapeType=Capsule, length=1.375, radius=1.125, center=<1.5541 0 0>, rotation=<rot 90 0 0>}]
 dyn_hair_front_r_05 [{shapeType=Capsule, length=2, radius=0.75, center=<1.5155 0 0>, rotation=<rot 90 0 0>}]
 dyn_hair_front_r_base [{shapeType=Box, extentX=0.75, extentY=0.5, extentZ=0.5}]
 dyn_hair_front_r_parent [{shapeType=Capsule, length=3.25, radius=1.5, center=<2.21 0 -0.0001>, rotation=<rot 90 0 0>}]
 dyn_hair_side_l_01 [{shapeType=Capsule, length=2.125, radius=2.875, center=<3.4512 0 0.0001>, rotation=<rot 90 0 0>}]
 dyn_hair_side_l_02 [{shapeType=Capsule, length=2.0359, radius=2.7141, center=<1.2845 0 -0.0001>, rotation=<rot 90 -9.4623 13.929>}]
 dyn_hair_side_l_03 [{shapeType=Capsule, length=1.236, radius=1.514, center=<0.955 0 -0.2532>, rotation=<rot 90 0 0>}]
 dyn_hair_side_l_04 [{shapeType=Capsule, length=1.3814, radius=1.9936, center=<2.1954 0.5831 -0.3198>, rotation=<rot 90 0 -13.4886>}]
 dyn_hair_side_l_base [{shapeType=Box, extentX=0.75, extentY=0.5, extentZ=0.5}]
 dyn_hair_side_l_parent [{shapeType=Capsule, length=2.25, radius=2, center=<2.6302 0 0.63>, rotation=<rot 90 0 0>}]
 dyn_hair_side_r_01 [{shapeType=Capsule, length=3.875, radius=1.5, center=<1.8633 0 0.0001>, rotation=<rot 90 0 0>}]
 dyn_hair_side_r_02 [{shapeType=Capsule, length=2.625, radius=1.75, center=<1.9399 0.4107 0.8238>, rotation=<rot 90 -159.444 -177.2946>}]
 dyn_hair_side_r_03 [{shapeType=Capsule, length=3.625, radius=1.125, center=<2.389 1.1723 0.1087>, rotation=<rot 75.0839 -179.9999 152.458>}]
 dyn_hair_side_r_base [{shapeType=Box, extentX=0.75, extentY=0.5, extentZ=0.5}]
 dyn_hair_side_r_parent [{shapeType=Capsule, length=2.125, radius=1.375, center=<1.7271 0 0>, rotation=<rot 90 0 0>}]
 dyn_head_offset [{shapeType=Box, extentX=0.5, extentY=0.5, extentZ=0.5}]
 head [{shapeType=Capsule, length=5.5, radius=5.875, center=<-0.0002 2.1581 -0.0324>, rotation=<rot 89.3496 179.9998 179.9998>}]
 neck_02 [{shapeType=Capsule, length=7.75, radius=5.375, center=<-1.6187 0.1345 -0.0184>, rotation=<rot 89.3496 179.9998 179.9998>}]
 spine_04 [{shapeType=Capsule, length=7.4066, radius=10.2626, shapeName=spine_04_capsule, center=<2.5514 0.7729 0>, rotation=<rot 0 -92.6845 0>}]
 spine_05 [{shapeType=Capsule, length=17.8161, radius=7.2701, shapeName=spine_05_capsule, center=<2.0326 -0.9928 -0.1551>}]
}
constraints [
 {bone1Name=dyn_hair_front_r_parent, bone2Name=dyn_hair_front_r_base, swing1LimitDegrees=15, swing2LimitDegrees=20, twistLimitDegrees=3}
 {bone1Name=dyn_hair_front_r_01, bone2Name=dyn_hair_front_r_parent, swing1LimitDegrees=15, swing2LimitDegrees=20, twistLimitDegrees=3}
 {bone1Name=dyn_hair_front_r_02, bone2Name=dyn_hair_front_r_01, swing1LimitDegrees=15, swing2LimitDegrees=20, twistLimitDegrees=3}
 {bone1Name=dyn_hair_front_r_03, bone2Name=dyn_hair_front_r_02, swing1LimitDegrees=20, swing2LimitDegrees=20, twistLimitDegrees=3}
 {bone1Name=dyn_hair_front_r_04, bone2Name=dyn_hair_front_r_03, swing1LimitDegrees=20, swing2LimitDegrees=20, twistLimitDegrees=3}
 {bone1Name=dyn_hair_front_r_05, bone2Name=dyn_hair_front_r_02, swing1LimitDegrees=20, swing2LimitDegrees=20, twistLimitDegrees=3}
 {bone1Name=dyn_hair_side_r_parent, bone2Name=dyn_hair_side_r_base, swing1LimitDegrees=10, swing2LimitDegrees=20, twistLimitDegrees=3}
 {bone1Name=dyn_hair_side_r_01, bone2Name=dyn_hair_side_r_parent, swing1LimitDegrees=10, swing2LimitDegrees=20, twistLimitDegrees=3}
 {bone1Name=dyn_hair_side_r_02, bone2Name=dyn_hair_side_r_01, swing1LimitDegrees=10, swing2LimitDegrees=20, twistLimitDegrees=3}
 {bone1Name=dyn_hair_side_r_03, bone2Name=dyn_hair_side_r_02, swing1LimitDegrees=10, swing2LimitDegrees=20, twistLimitDegrees=3}
 {
  bone1Name=dyn_hair_back_r_parent
  bone2Name=dyn_hair_back_r_base
  swing1LimitDegrees=5
  swing2LimitDegrees=25
  twistLimitDegrees=2
  twistMotion=Locked
 }
 {
  bone1Name=dyn_hair_back_r_01
  bone2Name=dyn_hair_back_r_parent
  swing1LimitDegrees=5
  swing2LimitDegrees=20
  twistLimitDegrees=2
  twistMotion=Locked
 }
 {
  bone1Name=dyn_hair_back_r_02
  bone2Name=dyn_hair_back_r_01
  swing1LimitDegrees=5
  swing2LimitDegrees=20
  twistLimitDegrees=2
  twistMotion=Locked
 }
 {
  bone1Name=dyn_hair_back_r_03
  bone2Name=dyn_hair_back_r_02
  swing1LimitDegrees=5
  swing2LimitDegrees=30
  twistLimitDegrees=3
  twistMotion=Locked
 }
 {bone1Name=dyn_hair_back_m_parent, bone2Name=dyn_hair_back_m_base, swing1LimitDegrees=5, swing2LimitDegrees=20, twistLimitDegrees=3}
 {bone1Name=dyn_hair_back_m_01, bone2Name=dyn_hair_back_m_parent, swing1LimitDegrees=5, swing2LimitDegrees=20, twistLimitDegrees=3}
 {bone1Name=dyn_hair_back_m_02, bone2Name=dyn_hair_back_m_01, swing1LimitDegrees=5, swing2LimitDegrees=20, twistLimitDegrees=3}
 {bone1Name=dyn_hair_back_m_03, bone2Name=dyn_hair_back_m_02, swing1LimitDegrees=5, swing2LimitDegrees=20, twistLimitDegrees=3}
 {bone1Name=dyn_hair_back_m_04, bone2Name=dyn_hair_back_m_03, swing1LimitDegrees=5, swing2LimitDegrees=20, twistLimitDegrees=3}
 {bone1Name=dyn_hair_back_m_05, bone2Name=dyn_hair_back_m_03, swing1LimitDegrees=5, swing2LimitDegrees=20, twistLimitDegrees=3}
 {bone1Name=dyn_hair_back_l_parent, bone2Name=dyn_hair_back_l_base, swing1LimitDegrees=5, swing2LimitDegrees=20, twistLimitDegrees=3}
 {bone1Name=dyn_hair_back_l_01, bone2Name=dyn_hair_back_l_parent, swing1LimitDegrees=5, swing2LimitDegrees=20, twistLimitDegrees=3}
 {bone1Name=dyn_hair_back_l_02, bone2Name=dyn_hair_back_l_01, swing1LimitDegrees=5, swing2LimitDegrees=20, twistLimitDegrees=3}
 {bone1Name=dyn_hair_back_l_03, bone2Name=dyn_hair_back_l_02, swing1LimitDegrees=5, swing2LimitDegrees=20, twistLimitDegrees=3}
 {bone1Name=dyn_hair_back_l_04, bone2Name=dyn_hair_back_l_01, swing1LimitDegrees=5, swing2LimitDegrees=20, twistLimitDegrees=3}
 {bone1Name=dyn_hair_back_l_05, bone2Name=dyn_hair_back_l_04, swing1LimitDegrees=5, swing2LimitDegrees=20, twistLimitDegrees=3}
 {bone1Name=dyn_hair_side_l_parent, bone2Name=dyn_hair_side_l_base, swing1LimitDegrees=10, swing2LimitDegrees=20, twistLimitDegrees=3}
 {bone1Name=dyn_hair_side_l_01, bone2Name=dyn_hair_side_l_parent, swing1LimitDegrees=15, swing2LimitDegrees=20, twistLimitDegrees=3}
 {bone1Name=dyn_hair_side_l_02, bone2Name=dyn_hair_side_l_01, swing1LimitDegrees=15, swing2LimitDegrees=20, twistLimitDegrees=3}
 {bone1Name=dyn_hair_side_l_03, bone2Name=dyn_hair_side_l_02, swing1LimitDegrees=30, swing2LimitDegrees=30, twistLimitDegrees=3}
 {bone1Name=dyn_hair_side_l_04, bone2Name=dyn_hair_side_l_03, swing1LimitDegrees=20, swing2LimitDegrees=30, twistLimitDegrees=3}
 {bone1Name=dyn_hair_front_l_parent, bone2Name=dyn_hair_front_l_base, swing1LimitDegrees=6, swing2LimitDegrees=20, twistLimitDegrees=3}
 {bone1Name=dyn_hair_front_l_01, bone2Name=dyn_hair_front_l_parent, swing1LimitDegrees=6, swing2LimitDegrees=20, twistLimitDegrees=3}
 {bone1Name=dyn_hair_front_l_02, bone2Name=dyn_hair_front_l_01, swing1LimitDegrees=6, swing2LimitDegrees=10, twistLimitDegrees=3}
 {bone1Name=dyn_hair_front_l_03, bone2Name=dyn_hair_front_l_01, swing1LimitDegrees=6, swing2LimitDegrees=10, twistLimitDegrees=3}
 {
  bone1Name=spine_05
  bone2Name=spine_04
  swing1LimitDegrees=45
  swing1Motion=Free
  swing2LimitDegrees=45
  swing2Motion=Free
  twistLimitDegrees=45
  twistMotion=Free
 }
 {
  bone1Name=clavicle_l
  bone2Name=spine_04
  swing1LimitDegrees=45
  swing1Motion=Free
  swing2LimitDegrees=45
  swing2Motion=Free
  twistLimitDegrees=45
  twistMotion=Free
 }
 {
  bone1Name=clavicle_r
  bone2Name=spine_04
  swing1LimitDegrees=45
  swing1Motion=Free
  swing2LimitDegrees=45
  swing2Motion=Free
  twistLimitDegrees=45
  twistMotion=Free
 }
 {
  bone1Name=neck_02
  bone2Name=spine_04
  swing1LimitDegrees=45
  swing1Motion=Free
  swing2LimitDegrees=45
  swing2Motion=Free
  twistLimitDegrees=45
  twistMotion=Free
 }
 {
  bone1Name=head
  bone2Name=spine_04
  swing1LimitDegrees=45
  swing1Motion=Free
  swing2LimitDegrees=45
  swing2Motion=Free
  twistLimitDegrees=45
  twistMotion=Free
 }
 {
  bone1Name=dyn_head_offset
  bone2Name=spine_04
  swing1LimitDegrees=45
  swing1Motion=Free
  swing2LimitDegrees=45
  swing2Motion=Free
  twistLimitDegrees=45
  twistMotion=Free
 }
 {
  bone1Name=dyn_hair_front_r_base
  bone2Name=spine_04
  swing1LimitDegrees=45
  swing1Motion=Free
  swing2LimitDegrees=45
  swing2Motion=Free
  twistLimitDegrees=45
  twistMotion=Free
 }
 {
  bone1Name=dyn_hair_side_r_base
  bone2Name=spine_04
  swing1LimitDegrees=45
  swing1Motion=Free
  swing2LimitDegrees=45
  swing2Motion=Free
  twistLimitDegrees=45
  twistMotion=Free
 }
 {
  bone1Name=dyn_hair_back_r_base
  bone2Name=spine_04
  swing1LimitDegrees=45
  swing1Motion=Free
  swing2LimitDegrees=45
  swing2Motion=Free
  twistLimitDegrees=45
  twistMotion=Free
 }
 {
  bone1Name=dyn_hair_back_m_base
  bone2Name=spine_04
  swing1LimitDegrees=45
  swing1Motion=Free
  swing2LimitDegrees=45
  swing2Motion=Free
  twistLimitDegrees=45
  twistMotion=Free
 }
 {
  bone1Name=dyn_hair_back_l_base
  bone2Name=spine_04
  swing1LimitDegrees=45
  swing1Motion=Free
  swing2LimitDegrees=45
  swing2Motion=Free
  twistLimitDegrees=45
  twistMotion=Free
 }
 {
  bone1Name=dyn_hair_side_l_base
  bone2Name=spine_04
  swing1LimitDegrees=45
  swing1Motion=Free
  swing2LimitDegrees=45
  swing2Motion=Free
  twistLimitDegrees=45
  twistMotion=Free
 }
 {
  bone1Name=dyn_hair_front_l_base
  bone2Name=spine_04
  swing1LimitDegrees=45
  swing1Motion=Free
  swing2LimitDegrees=45
  swing2Motion=Free
  twistLimitDegrees=45
  twistMotion=Free
 }
]
